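# Physics asset for the necklace on the F_MED_PaddleHarp_Gap female medium character body, set up for secondary motion of the necklace bones.
# Capsule bodies cover a base bone and a three-bone dynamic chain, with two capsules on the last link.
# Limited-swing, twist-free constraints join the chain link by link back to the base.
# Unconstrained neck and upper-spine capsules approximate the body, likely as colliders for the necklace.
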PhysicsAsset /BRCosmetics/Characters/Player/Female/Medium/Bodies/F_MED_PaddleHarp_Gap/Meshes/F_MED_PaddleHarp_Gap_Physics_Necklace
bodies {
 base_necklace_01_cr [{shapeType=Capsule, length=0.505, radius=0.505, shapeName=base_necklace_01_cr_capsule, rotation=<rot 0 -90 0>}]
 dyn_necklace_01_cr [{shapeType=Capsule, length=5, radius=1, shapeName=dyn_necklace_01_cr_capsule, center=<3 -0.4687 0>}]
 dyn_necklace_02_cr [{shapeType=Capsule, length=5, radius=1, shapeName=dyn_necklace_02_cr_capsule, center=<1 -0.779 0>}]
 dyn_necklace_03_cr [
  {shapeType=Capsule, length=5, radius=1, shapeName=dyn_necklace_03_cr_capsule, center=<1 -0.5 0>, rotation=<rot 90 0 0>}
  {shapeType=Capsule, length=5, radius=1, shapeName=dyn_necklace_03_cr_capsule, center=<4 -0.7119 0>}
 ]
 neck_01 [{shapeType=Capsule, length=5, radius=4.5, shapeName=neck_01_capsule, rotation=<rot 90 0 0>}]
 spine_05 [{shapeType=Capsule, length=13.3586, radius=10.7093, center=<-4.4737 -0.0234 -0.1873>, rotation=<rot 90 -36.8454 0>}]
}
constraints [
 {bone1Name=dyn_necklace_03_cr, bone2Name=dyn_necklace_02_cr, swing1LimitDegrees=15, swing2LimitDegrees=2, twistLimitDegrees=0}
 {bone1Name=dyn_necklace_02_cr, bone2Name=dyn_necklace_01_cr, swing1LimitDegrees=15, swing2LimitDegrees=2, twistLimitDegrees=0}
 {bone1Name=dyn_necklace_01_cr, bone2Name=base_necklace_01_cr, swing1LimitDegrees=10, swing2LimitDegrees=5, twistLimitDegrees=0}
]
