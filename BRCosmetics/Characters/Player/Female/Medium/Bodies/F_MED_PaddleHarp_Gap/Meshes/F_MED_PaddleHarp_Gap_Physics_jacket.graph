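# Physics asset for the jacket of the female medium PaddleHarp_Gap player cosmetic body, apparently for secondary motion of the jacket.
# Seven base anchor bones (back, three left, three right) connect to base_jacket with free swing and twist, all using tiny placeholder capsules.
# Each anchor carries one dyn_jacket bone with asymmetric swing limits and zero twist limit.
# The side dyn bones have larger fitted capsules, while the back one stays tiny.
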PhysicsAsset /BRCosmetics/Characters/Player/Female/Medium/Bodies/F_MED_PaddleHarp_Gap/Meshes/F_MED_PaddleHarp_Gap_Physics_jacket
bodies {
 base_jacket [{shapeType=Capsule, length=0.505, radius=0.505, shapeName=base_jacket_capsule, rotation=<rot 0 -90 0>}]
 base_jacket_back [{shapeType=Capsule, length=0.505, radius=0.505, shapeName=base_jacket_back_capsule, rotation=<rot 0 -90 0>}]
 base_jacket_l_0 [{shapeType=Capsule, length=0.505, radius=0.505, shapeName=base_jacket_l_0_capsule, rotation=<rot 0 -90 0>}]
 base_jacket_l_1 [{shapeType=Capsule, length=0.505, radius=0.505, shapeName=base_jacket_l_1_capsule, rotation=<rot 0 -90 0>}]
 base_jacket_l_2 [{shapeType=Capsule, length=0.505, radius=0.505, shapeName=base_jacket_l_2_capsule, rotation=<rot 0 -90 0>}]
 base_jacket_r_0 [{shapeType=Capsule, length=0.505, radius=0.505, shapeName=base_jacket_r_0_capsule, rotation=<rot 0 -90 0>}]
 base_jacket_r_1 [{shapeType=Capsule, length=0.505, radius=0.505, shapeName=base_jacket_r_1_capsule, rotation=<rot 0 -90 0>}]
 base_jacket_r_2 [{shapeType=Capsule, length=0.505, radius=0.505, shapeName=base_jacket_r_2_capsule, rotation=<rot 0 -90 0>}]
 dyn_jacket_back [{
   shapeType=Capsule
   length=0.505
   radius=0.505
   shapeName=dyn_jacket_back_capsule
   center=<0.7912 0.5633 1.3466>
   rotation=<rot 0 -79.4681 -83.7472>
  }]
 dyn_jacket_l_0 [{
   shapeType=Capsule
   length=5.4882
   radius=2.3627
   shapeName=dyn_jacket_l_0_capsule
   center=<4.396 -1.7902 0.2106>
   rotation=<rot 0 91.6319 -88.7503>
  }]
 dyn_jacket_l_1 [{
   shapeType=Capsule
   length=5.203
   radius=4.7942
   shapeName=dyn_jacket_l_1_capsule
   center=<4.2494 -0.1397 0.2845>
   rotation=<rot 0 -80.5947 -86.1648>
  }]
 dyn_jacket_l_2 [{
   shapeType=Capsule
   length=5.9707
   radius=5.2833
   shapeName=dyn_jacket_l_2_capsule
   center=<1.5722 0.303 -0.1492>
   rotation=<rot 0 -73.8956 -75.6145>
  }]
 dyn_jacket_r_0 [{
   shapeType=Capsule
   length=6.5732
   radius=2.451
   shapeName=dyn_jacket_r_0_capsule
   center=<3.1762 1.6773 0.22>
   rotation=<rot 0 91.8119 -89.178>
  }]
 dyn_jacket_r_1 [{
   shapeType=Capsule
   length=5.1373
   radius=4.7804
   shapeName=dyn_jacket_r_1_capsule
   center=<3.7967 0.2378 0.3086>
   rotation=<rot 0 -91.9889 -87.0459>
  }]
 dyn_jacket_r_2 [{
   shapeType=Capsule
   length=5.1003
   radius=4.8312
   shapeName=dyn_jacket_r_2_capsule
   center=<1.7583 -0.4169 -0.2879>
   rotation=<rot 0 -83.844 -72.7142>
  }]
}
constraints [
 {bone1Name=dyn_jacket_r_2, bone2Name=base_jacket_r_2, swing1LimitDegrees=10, swing2LimitDegrees=50, twistLimitDegrees=0}
 {
  bone1Name=base_jacket_back
  bone2Name=base_jacket
  swing1LimitDegrees=45
  swing1Motion=Free
  swing2LimitDegrees=45
  swing2Motion=Free
  twistLimitDegrees=45
  twistMotion=Free
 }
 {bone1Name=dyn_jacket_back, bone2Name=base_jacket_back, swing1LimitDegrees=10, swing2LimitDegrees=50, twistLimitDegrees=0}
 {
  bone1Name=base_jacket_l_0
  bone2Name=base_jacket
  swing1LimitDegrees=45
  swing1Motion=Free
  swing2LimitDegrees=45
  swing2Motion=Free
  twistLimitDegrees=45
  twistMotion=Free
 }
 {bone1Name=dyn_jacket_l_0, bone2Name=base_jacket_l_0, swing1LimitDegrees=10, swing2LimitDegrees=50, twistLimitDegrees=0}
 {
  bone1Name=base_jacket_l_1
  bone2Name=base_jacket
  swing1LimitDegrees=45
  swing1Motion=Free
  swing2LimitDegrees=45
  swing2Motion=Free
  twistLimitDegrees=45
  twistMotion=Free
 }
 {bone1Name=dyn_jacket_l_1, bone2Name=base_jacket_l_1, swing1LimitDegrees=10, swing2LimitDegrees=50, twistLimitDegrees=0}
 {
  bone1Name=base_jacket_l_2
  bone2Name=base_jacket
  swing1LimitDegrees=45
  swing1Motion=Free
  swing2LimitDegrees=45
  swing2Motion=Free
  twistLimitDegrees=45
  twistMotion=Free
 }
 {bone1Name=dyn_jacket_l_2, bone2Name=base_jacket_l_2, swing1LimitDegrees=10, swing2LimitDegrees=50, twistLimitDegrees=0}
 {
  bone1Name=base_jacket_r_0
  bone2Name=base_jacket
  swing1LimitDegrees=45
  swing1Motion=Free
  swing2LimitDegrees=45
  swing2Motion=Free
  twistLimitDegrees=45
  twistMotion=Free
 }
 {bone1Name=dyn_jacket_r_0, bone2Name=base_jacket_r_0, swing1LimitDegrees=10, swing2LimitDegrees=50, twistLimitDegrees=0}
 {
  bone1Name=base_jacket_r_1
  bone2Name=base_jacket
  swing1LimitDegrees=45
  swing1Motion=Free
  swing2LimitDegrees=45
  swing2Motion=Free
  twistLimitDegrees=45
  twistMotion=Free
 }
 {bone1Name=dyn_jacket_r_1, bone2Name=base_jacket_r_1, swing1LimitDegrees=10, swing2LimitDegrees=50, twistLimitDegrees=0}
 {
  bone1Name=base_jacket_r_2
  bone2Name=base_jacket
  swing1LimitDegrees=45
  swing1Motion=Free
  swing2LimitDegrees=45
  swing2Motion=Free
  twistLimitDegrees=45
  twistMotion=Free
 }
]
