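# Animation Blueprint for the F_MED_PaddleHarp_Gap female player body, adding secondary motion to its outfit.
# Layered parent, body and face poses feed a Control Rig aware of skydiving and targeting, then RigidBody necklace and jacket simulations.
# Weighted constraints position arm, boot and belt flap bones before spring-driven AnimDynamics produces the final pose.
# The event graph finds the head mesh, updates gravity override and zeroes RbanAlpha in the front end.
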AnimBlueprint /BRCosmetics/Characters/Player/Female/Medium/Bodies/F_MED_PaddleHarp_Gap/Meshes/F_MED_PaddleHarp_Gap_AnimBP
target_skeleton=&/F_MED_PaddleHarp_Gap_Skeleton
graphs [
 AnimGraph {nodes [
   Output Pose(Result=Use cached pose 'dyn_cache'()) {layerGroup=None}
   n1 = Control Rig!(Source=Use cached pose 'Start Cache'(), Alpha, IsSkydiving:bool=Get bIsSkyDiving(), IsTargeting:bool=Get bIsTargeting()) -> (Pose) {controlRigAssetReference {blueprintRigClass=&/F_MED_PaddleHarp_Gap_CtrlRig.F_MED_PaddleHarp_Gap_CtrlRig_C}}
   n2 = RigidBody!(ComponentPose=Local To Component(LocalPose=@n1), Alpha, OverrideWorldGravity:Vector="0,0,0") -> (Pose) {
    bClampLinearTranslationLimitToRefPose=true
    bOverrideWorldGravity=true
    overridePhysicsAsset=&/F_MED_PaddleHarp_Gap_Physics_Necklace
    simulationSpace=BaseBoneSpace
    baseBoneRef {boneName=spine_05}
    componentAppliedLinearAccClamp=<0 0 0>
    simSpaceSettings {
     maxAngularAcceleration=40
     maxAngularVelocity=10
     maxLinearAcceleration=1000
     maxLinearVelocity=300
     velocityScaleZ=0.4
     worldAlpha=1
     externalLinearDragV=<0.3 0.3 0.3>
    }
   }
   n3 = RigidBody(ComponentPose=@n2, OverrideWorldGravity="0,0,0") {
    bClampLinearTranslationLimitToRefPose=true
    bOverrideWorldGravity=true
    overridePhysicsAsset=&/F_MED_PaddleHarp_Gap_Physics_jacket
    simulationSpace=BaseBoneSpace
    alphaBoolBlend {blendOption=Linear}
    baseBoneRef {boneName=spine_04}
    componentAppliedLinearAccClamp=<0 0 0>
    simSpaceSettings {
     maxAngularAcceleration=40
     maxAngularVelocity=10
     maxLinearAcceleration=1000
     maxLinearVelocity=300
     velocityScaleZ=0.4
     worldAlpha=1
     externalLinearDragV=<0.3 0.3 0.3>
    }
   }
   Save cached pose 'control_rig_cache'(Pose=Component To Local(ComponentPose=@n3)) {cacheName=control_rig_cache}
   n4 = "Get Face Mesh "() -> ("Face Mesh ":Skeletal Mesh Component Object Reference)
   n5 = Copy Pose From Mesh() {bUseAttachedParent=true}
   n6 = Copy Pose From Mesh(SourceMeshComponent="Get Body Mesh "() -> ("Body Mesh ":Skeletal Mesh Component Object Reference)) {bUseAttachedParent=true}
   Save cached pose 'Start Cache'(Pose=Layered blend per bone(
     BasePose=Layered blend per bone(BasePose=@n5, BlendPoses_0=@n6) {layerSetup [{branchFilters [{boneName=clavicle_r}]}]}
     BlendPoses_0=Copy Pose From Mesh(SourceMeshComponent=@n4) {bUseAttachedParent=true}
    ) {layerSetup [{branchFilters [{boneName=FACIAL_R_ForeheadMid}]}]}) {cacheName=Start Cache}
   n7 = Anim Dynamics - Bone: dyn_flap_arm_l!(ComponentPose=Local To Component(LocalPose=Use cached pose 'predyn_cache'()), Alpha=1, GravityOverride:Vector="0,0,-980") -> (Pose) {
    angularDampingOverride=1
    angularSpringConstant=500
    bAngularSpring=true
    bOverrideAngularDamping=true
    bUseGravityOverride=true
    gravityScale=0.35
    simulationSpace=RootRelative
    alphaBoolBlend {blendOption=Linear}
    alphaScaleBiasClamp {interpSpeedDecreasing=2.5, interpSpeedIncreasing=8}
    boundBone {boneName=dyn_flap_arm_l}
    componentAppliedLinearAccClamp=<3000 3000 3000>
    componentLinearAccScale=<0.4 0.4 0.4>
    componentLinearVelScale=<0.3 0.3 0.3>
    gravityOverride {z=-980}
    physicsBodyDefinitions [{
      boundBone {boneName=dyn_flap_arm_l}
      boxExtents=<15 15 15>
      constraintSetup {angularLimitsMax=<0 8 0>, angularLimitsMin=<0 -8 0>, angularTarget=<1 0 0>}
      localJointOffset=<3 0 0>
     }]
    relativeSpaceBone {boneName=root}
    simSpaceSettings {maxAngularAcceleration=35, maxAngularVelocity=7, simSpaceAngularAlpha=1}
   }
   n8 = Anim Dynamics - Bone: dyn_flap_down_arm_r!(ComponentPose=@n7, Alpha=1, GravityOverride:Vector="0,0,-980") -> (Pose) {
    angularDampingOverride=1
    angularSpringConstant=500
    bAngularSpring=true
    bOverrideAngularDamping=true
    bUseGravityOverride=true
    gravityScale=0.35
    simulationSpace=RootRelative
    alphaBoolBlend {blendOption=Linear}
    alphaScaleBiasClamp {interpSpeedDecreasing=2.5, interpSpeedIncreasing=8}
    boundBone {boneName=dyn_flap_down_arm_r}
    componentAppliedLinearAccClamp=<3000 3000 3000>
    componentLinearAccScale=<0.4 0.4 0.4>
    componentLinearVelScale=<0.3 0.3 0.3>
    gravityOverride {z=-980}
    physicsBodyDefinitions [{
      boundBone {boneName=dyn_flap_down_arm_r}
      boxExtents=<18 18 18>
      constraintSetup {angularLimitsMax=<0 10 0>, angularLimitsMin=<0 -10 0>, angularTarget=<1 0 0>}
      localJointOffset=<2 0 0>
     }]
    relativeSpaceBone {boneName=root}
    simSpaceSettings {maxAngularAcceleration=35, maxAngularVelocity=7, simSpaceAngularAlpha=1}
   }
   n9 = Anim Dynamics - Bone: dyn_flap_upper_arm_r!(ComponentPose=@n8, Alpha=1, GravityOverride:Vector="0,0,-980") -> (Pose) {
    angularDampingOverride=1
    angularSpringConstant=500
    bAngularSpring=true
    bOverrideAngularDamping=true
    bUseGravityOverride=true
    gravityScale=0.35
    simulationSpace=RootRelative
    alphaBoolBlend {blendOption=Linear}
    alphaScaleBiasClamp {interpSpeedDecreasing=2.5, interpSpeedIncreasing=8}
    boundBone {boneName=dyn_flap_upper_arm_r}
    componentAppliedLinearAccClamp=<3000 3000 3000>
    componentLinearAccScale=<0.4 0.4 0.4>
    componentLinearVelScale=<0.3 0.3 0.3>
    gravityOverride {z=-980}
    physicsBodyDefinitions [{
      boundBone {boneName=dyn_flap_upper_arm_r}
      boxExtents=<15 15 15>
      constraintSetup {angularLimitsMax=<0 8 0>, angularLimitsMin=<0 -8 0>, angularTarget=<1 0 0>}
      localJointOffset=<3 0 0>
     }]
    relativeSpaceBone {boneName=root}
    simSpaceSettings {maxAngularAcceleration=35, maxAngularVelocity=7, simSpaceAngularAlpha=1}
   }
   n10 = Anim Dynamics - Bone: dyn_flap_down_boot_l!(ComponentPose=@n9, Alpha=1, GravityOverride:Vector="0,0,-980") -> (Pose) {
    angularDampingOverride=1
    angularSpringConstant=500
    bAngularSpring=true
    bOverrideAngularDamping=true
    bUseGravityOverride=true
    gravityScale=0.35
    simulationSpace=RootRelative
    alphaBoolBlend {blendOption=Linear}
    alphaScaleBiasClamp {interpSpeedDecreasing=2.5, interpSpeedIncreasing=8}
    boundBone {boneName=dyn_flap_down_boot_l}
    componentAppliedLinearAccClamp=<3000 3000 3000>
    componentLinearAccScale=<0.4 0.4 0.4>
    componentLinearVelScale=<0.3 0.3 0.3>
    gravityOverride {z=-980}
    physicsBodyDefinitions [{
      boundBone {boneName=dyn_flap_down_boot_l}
      boxExtents=<15 15 15>
      constraintSetup {angularLimitsMax=<0 8 0>, angularLimitsMin=<0 -8 0>, angularTarget=<1 0 0>}
      localJointOffset=<3 0 0>
     }]
    relativeSpaceBone {boneName=root}
    simSpaceSettings {maxAngularAcceleration=35, maxAngularVelocity=7, simSpaceAngularAlpha=1}
   }
   n11 = Anim Dynamics - Bone: dyn_flap_upper_boot_l!(ComponentPose=@n10, Alpha=1, GravityOverride:Vector="0,0,-980") -> (Pose) {
    angularDampingOverride=1
    angularSpringConstant=500
    bAngularSpring=true
    bOverrideAngularDamping=true
    bUseGravityOverride=true
    gravityScale=0.35
    simulationSpace=RootRelative
    alphaBoolBlend {blendOption=Linear}
    alphaScaleBiasClamp {interpSpeedDecreasing=2.5, interpSpeedIncreasing=8}
    boundBone {boneName=dyn_flap_upper_boot_l}
    componentAppliedLinearAccClamp=<3000 3000 3000>
    componentLinearAccScale=<0.4 0.4 0.4>
    componentLinearVelScale=<0.3 0.3 0.3>
    gravityOverride {z=-980}
    physicsBodyDefinitions [{
      boundBone {boneName=dyn_flap_upper_boot_l}
      boxExtents=<18 18 18>
      constraintSetup {angularLimitsMax=<0 10 0>, angularLimitsMin=<0 -10 0>, angularTarget=<1 0 0>}
      localJointOffset=<2 0 0>
     }]
    relativeSpaceBone {boneName=root}
    simSpaceSettings {maxAngularAcceleration=35, maxAngularVelocity=7, simSpaceAngularAlpha=1}
   }
   n12 = Anim Dynamics - Bone: dyn_flap_boot_r!(ComponentPose=@n11, Alpha=1, GravityOverride:Vector="0,0,-980") -> (Pose) {
    angularDampingOverride=1
    angularSpringConstant=500
    bAngularSpring=true
    bOverrideAngularDamping=true
    bUseGravityOverride=true
    gravityScale=0.35
    simulationSpace=RootRelative
    alphaBoolBlend {blendOption=Linear}
    alphaScaleBiasClamp {interpSpeedDecreasing=2.5, interpSpeedIncreasing=8}
    boundBone {boneName=dyn_flap_boot_r}
    componentAppliedLinearAccClamp=<3000 3000 3000>
    componentLinearAccScale=<0.4 0.4 0.4>
    componentLinearVelScale=<0.3 0.3 0.3>
    gravityOverride {z=-980}
    physicsBodyDefinitions [{
      boundBone {boneName=dyn_flap_boot_r}
      boxExtents=<15 15 15>
      constraintSetup {angularLimitsMax=<0 8 0>, angularLimitsMin=<0 -8 0>, angularTarget=<1 0 0>}
      localJointOffset=<3 0 0>
     }]
    relativeSpaceBone {boneName=root}
    simSpaceSettings {maxAngularAcceleration=35, maxAngularVelocity=7, simSpaceAngularAlpha=1}
   }
   n13 = Anim Dynamics - Bone: dyn_back_belt_flap!(ComponentPose=@n12, Alpha=1, GravityOverride:Vector="0,0,-980") -> (Pose) {
    angularDampingOverride=1
    angularSpringConstant=500
    bAngularSpring=true
    bOverrideAngularDamping=true
    bUseGravityOverride=true
    gravityScale=0.35
    simulationSpace=RootRelative
    alphaBoolBlend {blendOption=Linear}
    alphaScaleBiasClamp {interpSpeedDecreasing=2.5, interpSpeedIncreasing=8}
    boundBone {boneName=dyn_back_belt_flap}
    componentAppliedLinearAccClamp=<3000 3000 3000>
    componentLinearAccScale=<0.4 0.4 0.4>
    componentLinearVelScale=<0.3 0.3 0.3>
    gravityOverride {z=-980}
    physicsBodyDefinitions [{
      boundBone {boneName=dyn_back_belt_flap}
      boxExtents=<15 15 15>
      constraintSetup {angularLimitsMax=<0 0 0>, angularLimitsMin=<0 -25 0>, angularTarget=<1 0 0>}
      localJointOffset=<3 0 0>
     }]
    relativeSpaceBone {boneName=root}
    simSpaceSettings {maxAngularAcceleration=35, maxAngularVelocity=7, simSpaceAngularAlpha=1}
   }
   Save cached pose 'dyn_cache'(Pose=Component To Local(ComponentPose=@n13)) {cacheName=dyn_cache}
   n14 = "Constraint - dyn_flap_arm_l (PPP)"!(
    ComponentPose=Local To Component(LocalPose=Use cached pose 'control_rig_cache'())
    Alpha=1
    ConstraintWeights_0=0.8898
    ConstraintWeights_1:float=0.0558
    ConstraintWeights_2:float=0.0543
   ) -> (Pose) {
    boneToModify {boneName=dyn_flap_arm_l}
    constraintSetup [{targetBone {boneName=upperarm_twist_01_l}}, {targetBone {boneName=upperarm_l}}, {targetBone {boneName=upperarm_twist_02_l}}]
    constraintWeights [1,1,1]
   }
   n15 = "Constraint - dyn_flap_down_arm_r (PPPPPP)"!(
    ComponentPose=@n14
    Alpha=1
    ConstraintWeights_0=0.7251
    ConstraintWeights_1:float=0.1438
    ConstraintWeights_2:float=0.1193
    ConstraintWeights_3:float=0.0112
    ConstraintWeights_4:float=0.0004
    ConstraintWeights_5:float=0.0003
   ) -> (Pose) {
    boneToModify {boneName=dyn_flap_down_arm_r}
    constraintSetup [
     {targetBone {boneName=upperarm_twist_01_r}}
     {targetBone {boneName=upperarm_r}}
     {targetBone {boneName=upperarm_twist_02_r}}
     {targetBone {boneName=spine_04}}
     {targetBone {boneName=deform_elbow_in_r}}
     {targetBone {boneName=lowerarm_r}}
    ]
    constraintWeights [1,1,1,1,1,1]
   }
   n16 = "Constraint - dyn_flap_upper_arm_r (PPPPPPPP)"!(
    ComponentPose=@n15
    Alpha=1
    ConstraintWeights_0=0.6651
    ConstraintWeights_1:float=0.1649
    ConstraintWeights_2:float=0.1494
    ConstraintWeights_3:float=0.0055
    ConstraintWeights_4:float=0.0046
    ConstraintWeights_5:float=0.0042
    ConstraintWeights_6:float=0.0033
    ConstraintWeights_7:float=0.0029
   ) -> (Pose) {
    boneToModify {boneName=dyn_flap_upper_arm_r}
    constraintSetup [
     {targetBone {boneName=upperarm_twist_01_r}}
     {targetBone {boneName=upperarm_r}}
     {targetBone {boneName=upperarm_twist_02_r}}
     {targetBone {boneName=deform_upperarm_fr_r}}
     {targetBone {boneName=clavicle_r}}
     {targetBone {boneName=deform_upperarm_up_r}}
     {targetBone {boneName=deform_upperarm_bk_r}}
     {targetBone {boneName=spine_05}}
    ]
    constraintWeights [1,1,1,1,1,1,1,1]
   }
   n17 = "Constraint - dyn_flap_down_boot_l (PPPP)"!(
    ComponentPose=@n16
    Alpha=1
    ConstraintWeights_0=0.7363
    ConstraintWeights_1:float=0.1454
    ConstraintWeights_2:float=0.1049
    ConstraintWeights_3:float=0.0134
   ) -> (Pose) {
    boneToModify {boneName=dyn_flap_down_boot_l}
    constraintSetup [
     {targetBone {boneName=thigh_l}}
     {targetBone {boneName=calf_l}}
     {targetBone {boneName=deform_knee_out_l}}
     {targetBone {boneName=deform_knee_in_l}}
    ]
    constraintWeights [1,1,1,1]
   }
   n18 = "Constraint - dyn_flap_upper_boot_l (PPPPP)"!(
    ComponentPose=@n17
    Alpha=1
    ConstraintWeights_0=0.9182
    ConstraintWeights_1:float=0.0382
    ConstraintWeights_2:float=0.0321
    ConstraintWeights_3:float=0.0111
    ConstraintWeights_4:float=0.0003
   ) -> (Pose) {
    boneToModify {boneName=dyn_flap_upper_boot_l}
    constraintSetup [
     {targetBone {boneName=thigh_l}}
     {targetBone {boneName=calf_l}}
     {targetBone {boneName=deform_knee_out_l}}
     {targetBone {boneName=thigh_twist_01_l}}
     {targetBone {boneName=deform_knee_in_l}}
    ]
    constraintWeights [1,1,1,1,1]
   }
   n19 = "Constraint - dyn_flap_boot_r (PPPPP)"!(
    ComponentPose=@n18
    Alpha=1
    ConstraintWeights_0=0.8699
    ConstraintWeights_1:float=0.0693
    ConstraintWeights_2:float=0.0556
    ConstraintWeights_3:float=0.0045
    ConstraintWeights_4:float=0.0007
   ) -> (Pose) {
    boneToModify {boneName=dyn_flap_boot_r}
    constraintSetup [
     {targetBone {boneName=thigh_r}}
     {targetBone {boneName=calf_r}}
     {targetBone {boneName=deform_knee_out_r}}
     {targetBone {boneName=deform_knee_in_r}}
     {targetBone {boneName=thigh_twist_01_r}}
    ]
    constraintWeights [1,1,1,1,1]
   }
   n20 = "Constraint - dyn_back_belt_flap (PPPPPPP)"!(
    ComponentPose=@n19
    Alpha=1
    ConstraintWeights_0=0.8776
    ConstraintWeights_1:float=0.0844
    ConstraintWeights_2:float=0.0236
    ConstraintWeights_3:float=0.0101
    ConstraintWeights_4:float=0.0038
    ConstraintWeights_5:float=0.0002
    ConstraintWeights_6:float=0.0002
   ) -> (Pose) {
    boneToModify {boneName=dyn_back_belt_flap}
    constraintSetup [
     {targetBone {boneName=pelvis}}
     {targetBone {boneName=deform_glute_r}}
     {targetBone {boneName=spine_02}}
     {targetBone {boneName=spine_01}}
     {targetBone {boneName=thigh_twist_01_r}}
     {targetBone {boneName=spine_03}}
     {targetBone {boneName=spine_04}}
    ]
    constraintWeights [1,1,1,1,1,1,1]
   }
   Save cached pose 'predyn_cache'(Pose=Component To Local(ComponentPose=@n20)) {cacheName=predyn_cache}
  ]}
 EventGraph {nodes [
   Event BlueprintUpdateAnimation()
   > n1 = Sequence()
   > n2 = "Gravity Override State Function (Parameter Asset) (Nativized)"(
    CharacterPartAnimInstance=Self-Reference()
    GameTimeInSeconds=GetGameTimeInSeconds()
    Parameters=&/F_MED_PaddleHarp_Gap_Parameters
   )
   > Set GravityOverride(GravityOverride=@n2)
   n3 = Branch(execute=@n1.then_1, Condition=NOT Boolean(A=Get bIsInFrontEnd()))
   > Set RbanAlpha(RbanAlpha:double=1) -> (Output_Get:double)
   Set RbanAlpha(execute=@n3.else, RbanAlpha=0)
   Event OnCharacterCustomizationLoaded()
   > n4 = Cast To PlayerPawn_Athena(Object=TryGetPawnOwner())
   > "Set Face Mesh "("Face Mesh ":Skeletal Mesh Component Object Reference=GetSkeletalMeshForPartType(self=@n4, PartType=Head)) -> (Output_Get:Skeletal Mesh Component Object Reference)
   Event BlueprintInitializeAnimation()
   > Set SkelMesh(SkelMesh=GetOwnerSkeletalMeshComponent())
  ]}
]
variables [
 "Body Mesh " {type=SkeletalMeshComponent}
 "Face Mesh " {type=SkeletalMeshComponent}
 GravityOverride {type=Vector}
 RbanAlpha {type=number}
 RbanFEAlpha {type=number}
 SkelMesh {type=SkeletalMeshComponent}
]
