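# Control Rig for the F_MED_PaddleHarp_Gap face accessory that procedurally adjusts hair bones.
# One function parent-constrains three added hair bones to weighted bones or the head.
# Another swing-twist decomposes base_hair relative to a spine-parented control, remapping swing through curves to offset right-side hair bones; the stored twist is unread.
# It also multiplies head and right-upperarm spherical pose readers to weight a corrective transform on right-side hair.
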ControlRigBlueprint /BRCosmetics/Characters/Player/Female/Medium/Bodies/F_MED_PaddleHarp_Gap/Meshes/Parts/F_MED_PaddleHarp_Gap_FaceAcc_CtrlRig
graphs [
 New Function_ContainedGraph {nodes [
   RigVMFunctionEntryNode()
   > RigVMFunction_Sequence = RigVMAggregateNode()
   n1 = &RigUnit_SphericalPoseReader(
    ActiveRegionScaleFactors=(PositiveWidth=1,NegativeWidth=0.5,PositiveHeight=0.75,NegativeHeight=0.75)
    ActiveRegionSize=0.1
    Debug=(bDrawDebug=false,bDraw2D=False,bDrawLocalAxes=False,DebugScale=25,DebugSegments=20,DebugThickness=0.25)
    DriverAxis=(X=0,Y=1,Z=0)
    DriverItem=(Type=Bone,Name="head")
    FalloffRegionScaleFactors=(PositiveWidth=1,NegativeWidth=0.5,PositiveHeight=1,NegativeHeight=1)
    FalloffSize=0.2
    FlipHeightScaling=False
    FlipWidthScaling=False
    OptionalParentItem=(Type=Bone,Name="None")
    RotationOffset=(X=90,Y=0,Z=-30)
    ExecutePin=@RigVMFunction_Sequence.C
   )
   > n2 = &RigUnit_SphericalPoseReader(
    ActiveRegionScaleFactors=(PositiveWidth=1,NegativeWidth=1,PositiveHeight=1,NegativeHeight=1)
    ActiveRegionSize=0.025
    Debug=(bDrawDebug=false,bDraw2D=False,bDrawLocalAxes=False,DebugScale=25,DebugSegments=20,DebugThickness=0.25)
    DriverAxis=(X=-1,Y=0,Z=0)
    DriverItem=(Type=Bone,Name="upperarm_r")
    FalloffRegionScaleFactors=(PositiveWidth=1,NegativeWidth=1,PositiveHeight=1,NegativeHeight=1)
    FalloffSize=0.1
    FlipHeightScaling=False
    FlipWidthScaling=False
    OptionalParentItem=(Type=Bone,Name="None")
    RotationOffset=(X=-115,Y=0,Z=25)
   )
   Multiply = &RigVMFunction_MathDoubleMul(A=@n2.OutputParam, B=@n1.OutputParam)
   > &RigUnit_ModifyTransforms(
    ItemToModify="((Item=(Type=Bone,Name=\"base_hair_font_side_r\"),Transform=(Rotation=(X=0,Y=0.1305,Z=0,W=0.9914),Translation=(X=0,Y=0,Z=0.5),Scale3D=(X=1,Y=1,Z=1))),(Item=(Type=Bone,Name=\"dyn_hair_side_r_01\"),Transform=(Rotation=(X=0,Y=0.1305,Z=0,W=0.9914),Translation=(X=0 ...(truncated)"
    WeightMinimum=0
    Weight=@Multiply.Result
   )
   RigVMFunctionReturnNode(ExecuteContext=@RigVMFunction_Sequence.D)
   n3 = &RigVMFunction_MathQuaternionSwingTwist(
    TwistAxis=(X=1,Y=0,Z=0)
    Input=&RigUnit_GetRelativeTransformForItem(Child=(Type=Bone,Name="base_hair"), Parent=(Type=Control,Name="base_hair_ctrl"), bChildInitial=False, bParentInitial=False).RelativeTransform.Rotation
   )
   RigVMVariableNode(Value=&RigVMFunction_MathQuaternionToEuler(RotationOrder=ZYX, Value=@n3.Swing).Result, ExecuteContext=@RigVMFunction_Sequence.A) {variable=headSwing_rot}
   > RigVMVariableNode(Value=&RigVMFunction_MathQuaternionToEuler(RotationOrder=ZYX, Value=@n3.Twist).Result) {variable=headSwing_twist}
   n4 = RigVMVariableNode() {variable=headSwing_rot}
   n5 = &RigVMFunction_AnimEvalRichCurve(
    Curve="(EditorCurveData=(Keys=((InterpMode=RCIM_Cubic,TangentMode=RCTM_User,ArriveTangent=0.2796,LeaveTangent=0.2796),(InterpMode=RCIM_Cubic,TangentMode=RCTM_User,Time=0.5061,Value=0.359,ArriveTangent=1.2761,LeaveTangent=1.2761),(InterpMode=RCIM_Cubic,TangentMode ...(truncated)"
    SourceMaximum=-140
    SourceMinimum=0
    TargetMaximum=-44
    TargetMinimum=0
    Value=@n4.Value.Y
   )
   n6 = &RigVMFunction_AnimEvalRichCurve(
    Curve="(EditorCurveData=(Keys=((InterpMode=RCIM_Cubic,TangentMode=RCTM_User,ArriveTangent=0.2796,LeaveTangent=0.2796),(InterpMode=RCIM_Cubic,TangentMode=RCTM_User,Time=0.4909,Value=0.5539,ArriveTangent=1.2761,LeaveTangent=1.2761),(InterpMode=RCIM_Cubic,TangentMod ...(truncated)"
    SourceMaximum=-140
    SourceMinimum=0
    TargetMaximum=-44
    TargetMinimum=0
    Value=@n4.Value.Z
   )
   n7 = &RigVMFunction_AlphaInterp(
    InterpSpeedDecreasing=10
    InterpSpeedIncreasing=10
    Scale=1
    bClampResult=false
    bInterpResult=false
    bMapRange=true
    Value=&RigVMFunction_MathDoubleAdd(A=@n5.Result, B=@n6.Result).Result
   )
   n8 = &RigVMFunction_MathDoubleMul(B=0.1, A=@n7.Result)
   n9 = &RigVMFunction_MathQuaternionFromEuler(
    Euler=(X=0,Y=0,Z=0)
    Euler.Y=&RigVMFunction_MathDoubleMul(B=0.3, A=@n7.Result).Result
    Euler.Z=&RigVMFunction_MathDoubleMul(B=-0.3, A=@n7.Result).Result
   )
   &RigUnit_ModifyTransforms(
    ItemToModify="((Item=(Type=Bone,Name=\"base_hair_back_r\"),Transform=(Rotation=(X=0,Y=0,Z=0,W=1),Translation=(X=0,Y=0,Z=0),Scale3D=(X=1,Y=1,Z=1))),(Item=(Type=Bone,Name=\"base_hair_side_r\"),Transform=(Rotation=(X=0,Y=0,Z=0,W=1),Translation=(X=0,Y=0,Z=0),Scale3D=(X=1,Y=1,Z= ...(truncated)"
    Weight=1
    WeightMinimum=0
    ItemToModify.0.Transform.Translation.X=@n8.Result
    ItemToModify.3.Transform.Translation.X=@n8.Result
    ItemToModify.0.Transform.Rotation=&RigVMFunction_MathQuaternionFromEuler(
     Euler=(X=0,Y=0,Z=0)
     Euler.Z=&RigVMFunction_MathDoubleMul(B=0.3, A=@n7.Result).Result
     Euler.Y=&RigVMFunction_MathDoubleMul(B=-0.3, A=@n7.Result).Result
    ).Result
    ItemToModify.1.Transform.Rotation=@n9.Result
    ItemToModify.2.Transform.Rotation=@n9.Result
    ItemToModify.3.Transform.Rotation=@n9.Result
    ExecutePin=@RigVMFunction_Sequence.B
   )
  ]}
 New Function_ContainedGraph {nodes [
   RigVMFunctionEntryNode()
   > &RigUnit_ParentConstraint(
    AdvancedSettings=(RotationOrderForFilter=XZY)
    Child=(Type=Bone,Name="dyn_hair_add_01")
    Parents="((Item=(Type=Bone,Name=\"dyn_hair_side_r_02\"),Weight=0.0004),(Item=(Type=Bone,Name=\"dyn_hair_side_r_01\"),Weight=0.0013),(Item=(Type=Bone,Name=\"dyn_hair_back_r_03\"),Weight=0.0229),(Item=(Type=Bone,Name=\"dyn_hair_back_r_01\"),Weight=0.0301),(Item=(Type=Bone,Na ...(truncated)"
    Weight=1
   )
   > &RigUnit_ParentConstraint(
    AdvancedSettings=(RotationOrderForFilter=XZY)
    Child=(Type=Bone,Name="dyn_hair_add_02")
    Parents=((Item=(Type=Bone,Name="head"),Weight=1.0))
    Weight=1
   )
   > &RigUnit_ParentConstraint(
    AdvancedSettings=(RotationOrderForFilter=XZY)
    Child=(Type=Bone,Name="dyn_hair_add_03")
    Parents=((Item=(Type=Bone,Name="head"),Weight=1.0))
    Weight=1
   )
   > RigVMFunctionReturnNode()
  ]}
 RigVMFunctionLibrary {nodes [
   Hair_Setup = RigVMCollapseNode(ExecuteContext=())
   add_hair = RigVMCollapseNode(ExecuteContext=())
  ]}
 RigVMFunction_Sequence_1_ContainedGraph {nodes [
   n1 = RigVMFunctionEntryNode()
   RigVMFunction_Sequence = &RigVMFunction_Sequence(ExecuteContext=@n1.ExecuteContext)
   n2 = &RigVMFunction_Sequence(ExecuteContext=@RigVMFunction_Sequence.B)
   n3 = &RigVMFunction_Sequence(ExecuteContext=@n2.B)
   RigVMFunctionReturnNode(A=@RigVMFunction_Sequence.A, B=@n2.A, C=@n3.A, D=@n3.B)
  ]}
 RigVMModel {nodes [
   RigUnit_BeginExecution = &RigUnit_BeginExecution()
   add_hair = RigVMFunctionReferenceNode(ExecuteContext=@RigUnit_BeginExecution.ExecutePin)
   > Hair_Setup = RigVMFunctionReferenceNode()
  ]}
]
hierarchy [
 root {loc=<0 0 0>, rot=<0 0 0 1>}
 pelvis {parent=root, loc=<0 -0.5301 98.0571>, rot=<0.0127 0.707 -0.0127 -0.707>}
 spine_01 {parent=pelvis, loc=<6.1208 -0.9328 0>, rot=<0 0 0.034 -0.9994>}
 spine_02 {parent=spine_01, loc=<6.5997 -0.2095 0>, rot=<0 0 0.0238 -0.9997>}
 spine_03 {parent=spine_02, loc=<6.4914 -0.1237 0>, rot=<0 0 0.0149 -0.9999>}
 spine_04 {parent=spine_03, loc=<6.9598 0.0794 0>, rot=<0 0 -0.0132 -0.9999>}
 spine_05 {parent=spine_04, loc=<6.9987 0 0>, rot=<0 0 -0.0015 -1>}
 neck_01 {parent=spine_05, loc=<11.3138 0.1529 0>, rot=<0 0 -0.2003 -0.9797>}
 neck_02 {parent=neck_01, loc=<4.5627 0 0>, rot=<0 0 0.0504 -0.9987>}
 head {parent=neck_02, loc=<5.3914 0 0>, rot=<0 0 0.115 -0.9934>}
 base_earring_l {parent=head, loc=<-0.1587 0.3817 -6.1597>, rot=<-0.004 -0.2954 0.9554 0.0006>}
 dyn_earring_l {parent=base_earring_l, loc=<0 0 0>, rot=<0 0 0 1>}
 base_earring_r {parent=head, loc=<-0.1535 0.4159 6.1755>, rot=<-0.0086 0.2474 0.9689 0.0055>}
 dyn_earring_r {parent=base_earring_r, loc=<0 0 0>, rot=<0 0 0 1>}
 base_hair {parent=head, loc=<0 0 0>, rot=<0 0 0 1>}
 base_hair_back_r {parent=base_hair, loc=<9.8531 -7.2126 6.3505>, rot=<0.0389 -0.4191 0.9027 0.0896>}
 dyn_hair_back_r_01 {parent=base_hair_back_r, loc=<0 0 0>, rot=<0 0 0 1>}
 dyn_hair_back_r_02 {parent=dyn_hair_back_r_01, loc=<6.0862 -0.9565 2.1965>, rot=<0 0.012 0 -0.9999>}
 dyn_hair_back_r_03 {parent=dyn_hair_back_r_02, loc=<4.7494 -0.2121 -0.6934>, rot=<0 -0.0082 0 -1>}
 base_hair_side_r {parent=base_hair, loc=<9.7466 0.9835 7.4969>, rot=<0.1827 -0.0297 0.9826 -0.0141>}
 dyn_hair_side_r_01 {parent=base_hair_side_r, loc=<0 0 0>, rot=<0 0 0 1>}
 dyn_hair_side_r_02 {parent=dyn_hair_side_r_01, loc=<8.187 1.3035 -0.1362>, rot=<0 -0.2051 0 -0.9787>}
 dyn_hair_side_r_03 {parent=dyn_hair_side_r_02, loc=<4.5797 0.0856 -0.5188>, rot=<0 -0.0816 0 -0.9967>}
 base_hair_font_side_r {parent=base_hair, loc=<7.6524 8.6679 6.1387>, rot=<0.1595 0.2344 0.9584 -0.032>}
 dyn_hair_front_side_r_01 {parent=base_hair_font_side_r, loc=<0 0 0>, rot=<0 0 0 1>}
 dyn_hair_front_side_r_02 {parent=dyn_hair_front_side_r_01, loc=<7.6668 0.8437 -0.5406>, rot=<0.0307 -0.2848 0.0861 -0.9542>}
 dyn_hair_front_side_r_03 {parent=dyn_hair_front_side_r_02, loc=<6.7462 0.4285 0.4219>, rot=<0 -0.0558 0 -0.9984>}
 base_hair_front_01_r {parent=base_hair, loc=<4.7787 10.2646 2.8939>, rot=<0.0843 0.2286 0.9679 -0.0612>}
 dyn_hair_front_01_r {parent=base_hair_front_01_r, loc=<0 0 0>, rot=<0 0 0 1>}
 base_hair_add_03 {parent=base_hair, loc=<5.485 -6.787 -9.7293>, rot=<0.5286 -0.5029 0.4644 -0.502>}
 dyn_hair_add_03 {parent=base_hair_add_03, loc=<0 0 0>, rot=<0 0 0 1>}
 base_hair_add_02 {parent=base_hair, loc=<9.8955 -6.8621 -4.6855>, rot=<0.4562 -0.5318 0.6446 -0.3059>}
 dyn_hair_add_02 {parent=base_hair_add_02, loc=<0 0 0>, rot=<0 0 0 1>}
 base_hair_add_01 {parent=base_hair, loc=<8.0424 -6.4575 5.5112>, rot=<0.3621 -0.5314 0.7052 -0.2986>}
 dyn_hair_add_01 {parent=base_hair_add_01, loc=<0 0 0>, rot=<0 0 0 1>}
 base_hair_add_04 {parent=base_hair, loc=<5.2559 -1.3545 -9.4907>, rot=<-0.0564 0.6258 0.7555 -0.1858>}
 dyn_hair_add_04 {parent=base_hair_add_04, loc=<0 0 0>, rot=<0 0 0 1>}
 dyn_hair_add_05 {parent=dyn_hair_add_04, loc=<5.8632 -0.6683 0.3957>, rot=<0.9995 -0.0004 0.0045 -0.0315>}
 base_hair_back_01 {parent=base_hair, loc=<4.3597 -8.2883 4.1248>, rot=<0.3924 -0.3833 0.8309 0.0937>}
 dyn_hair_back_01 {parent=base_hair_back_01, loc=<0 0 0>, rot=<0 0 0 1>}
 base_hair_back_02 {parent=base_hair, loc=<4.576 -10.3693 1.5444>, rot=<0.3574 -0.4741 0.7974 0.1078>}
 dyn_hair_back_02 {parent=base_hair_back_02, loc=<0 0 0>, rot=<0 0 0 1>}
 base_hair_back_03 {parent=base_hair, loc=<3.8853 -10.802 0.1061>, rot=<0.413 -0.4437 0.7943 -0.0394>}
 dyn_hair_back_03 {parent=base_hair_back_03, loc=<0 0 0>, rot=<0 0 0 1>}
 base_hair_back_04 {parent=base_hair, loc=<4.2841 -11.0358 -1.0268>, rot=<0.0649 -0.855 -0.1917 -0.4775>}
 dyn_hair_back_04 {parent=base_hair_back_04, loc=<0 0 0>, rot=<0 0 0 1>}
 base_hair_back_05 {parent=base_hair, loc=<5.1168 -9.8127 -2.9902>, rot=<-0.0666 -0.6713 0.6395 -0.3686>}
 dyn_hair_back_05 {parent=base_hair_back_05, loc=<0 0 0>, rot=<0 0 0 1>}
 base_hair_back_06 {parent=base_hair, loc=<7.0969 -9.4075 -4.6042>, rot=<-0.335 -0.728 0.5441 -0.2486>}
 dyn_hair_back_06 {parent=base_hair_back_06, loc=<0 0 0>, rot=<0 0 0 1>}
 base_earing_l {parent=head, loc=<-0.1519 -0.2459 -6.3404>, rot=<0.4979 0.5021 0.5021 0.4979>}
 dyn_earing_l {parent=base_earing_l, loc=<0 0 0>, rot=<0 0 0 1>}
 base_earing_r {parent=head, loc=<-0.1552 -0.2455 6.3341>, rot=<0.4979 0.5021 0.5021 0.4979>}
 dyn_earing_r {parent=base_earing_r, loc=<0 0 0>, rot=<0 0 0 1>}
 base_neck {parent=neck_01, loc=<0 0 0>, rot=<0 0 0.0504 -0.9987>}
 base_neck_side_l {parent=base_neck, loc=<1.0083 -0.0025 -6.7339>, rot=<0.0374 -0.778 0.0466 -0.6254>}
 dyn_neck_side_l {parent=base_neck_side_l, loc=<0 0 0>, rot=<0 0 0 1>}
 base_neck_front_l {parent=base_neck, loc=<-0.1199 4.2639 -7.2486>, rot=<0.1342 -0.7571 -0.0022 -0.6393>}
 dyn_neck_front_l {parent=base_neck_front_l, loc=<0 0 0>, rot=<0 0 0 1>}
 base_neck_backside_l {parent=base_neck, loc=<1.407 -4.0825 -5.1488>, rot=<0.2309 -0.7145 0.2338 -0.6177>}
 dyn_neck_backside_l {parent=base_neck_backside_l, loc=<0 0 0>, rot=<0 0 0 1>}
 base_neck_backside_r {parent=base_neck, loc=<1.3531 -4.0478 5.0486>, rot=<0.601 -0.1852 0.7578 -0.174>}
 dyn_neck_backside_r {parent=base_neck_backside_r, loc=<0 0 0>, rot=<0 0 0 1>}
 base_neck_side_r {parent=base_neck, loc=<0.9288 -0.1638 6.4621>, rot=<0.6361 -0.076 0.7581 -0.1216>}
 dyn_neck_side_r {parent=base_neck_side_r, loc=<0 0 0>, rot=<0 0 0 1>}
 base_neck_front_r {parent=base_neck, loc=<-0.1395 4.2638 7.363>, rot=<0.6762 0.1105 0.7222 -0.0943>}
 dyn_neck_front_r {parent=base_neck_front_r, loc=<0 0 0>, rot=<0 0 0 1>}
 base_neck_back_cr {parent=base_neck, loc=<1.3935 -6.1858 0>, rot=<0.4591 -0.5378 0.5378 -0.4591>}
 dyn_neck_back_cr {parent=base_neck_back_cr, loc=<0 0 0>, rot=<0 0 0 1>}
 clavicle_l {parent=spine_05, loc=<5.4831 -0.1053 -3.2946>, rot=<0.0465 -0.7289 -0.009 -0.6829>}
 upperarm_l {parent=clavicle_l, loc=<9.8658 -1.2136 -0.4105>, rot=<0.0164 -0.3245 -0.0269 -0.9454>}
 lowerarm_l {parent=upperarm_l, loc=<26.465 -0.8345 0.259>, rot=<0.0036 0.0161 -0.2298 -0.9731>}
 hand_l {parent=lowerarm_l, loc=<22.7355 -0.3203 -0.6105>, rot=<-0.6784 -0.0223 0.0735 -0.7307>}
 hand_attach_l {parent=hand_l, loc=<8.5 -3 0>, rot=<0.7071 0.7071 0 0>}
 weapon_l {parent=hand_l, loc=<-0.1974 -3.4111 0.3811>, rot=<0 0 0 1>}
 thumb_01_l {parent=hand_l, loc=<1.3118 -0.3191 -2.459>, rot=<0.6546 -0.4042 0.0664 -0.6354>}
 thumb_02_l {parent=thumb_01_l, loc=<5.9336 -0.0448 0.0401>, rot=<-0.0037 0.0213 0.298 -0.9543>}
 thumb_03_l {parent=thumb_02_l, loc=<2.2223 0.3682 -0.0907>, rot=<0.0118 -0.013 0.0066 -0.9998>}
 index_metacarpal_l {parent=hand_l, loc=<1.6535 0.0183 -2.96>, rot=<0.0158 -0.0124 -0.0173 -0.9996>}
 index_01_l {parent=index_metacarpal_l, loc=<8.3008 0.0357 -0.0524>, rot=<-0.0044 0.0485 0.0924 -0.9945>}
 index_02_l {parent=index_01_l, loc=<3.3587 -0.3004 0.0452>, rot=<-0.002 0.0027 0.3225 -0.9466>}
 index_03_l {parent=index_02_l, loc=<2.5578 -0.123 -0.0252>, rot=<-0.002 0.0067 0.2558 -0.9667>}
 middle_metacarpal_l {parent=hand_l, loc=<2.1856 -0.1725 -0.5984>, rot=<0.006 0.0168 -0.0329 -0.9993>}
 middle_01_l {parent=middle_metacarpal_l, loc=<7.3275 -0.2447 0.0129>, rot=<-0.0208 0.0042 0.0858 -0.9961>}
 middle_02_l {parent=middle_01_l, loc=<4.0123 0.1401 0.1175>, rot=<-0.0218 0.0258 0.3512 -0.9357>}
 middle_03_l {parent=middle_02_l, loc=<2.6077 0.0534 0.1373>, rot=<0.0082 0.0188 0.1708 -0.9851>}
 ring_metacarpal_l {parent=hand_l, loc=<2.2074 -0.3625 1.3638>, rot=<-0.0284 0.0722 -0.0458 -0.9959>}
 ring_01_l {parent=ring_metacarpal_l, loc=<7.0295 -0.6197 -0.1446>, rot=<-0.0489 -0.0252 0.1892 -0.9804>}
 ring_02_l {parent=ring_01_l, loc=<3.6174 0.4647 -0.1027>, rot=<0 0 0.2806 -0.9598>}
 ring_03_l {parent=ring_02_l, loc=<2.4152 0.114 0.0199>, rot=<0 0 0.1189 -0.9929>}
 pinky_metacarpal_l {parent=hand_l, loc=<1.7616 -0.9753 3.3814>, rot=<-0.0397 0.0814 -0.069 -0.9935>}
 pinky_01_l {parent=pinky_metacarpal_l, loc=<7.1709 -0.4831 -0.1094>, rot=<-0.0654 -0.0035 0.2242 -0.9723>}
 pinky_02_l {parent=pinky_01_l, loc=<2.8274 0.0893 -0.1043>, rot=<0.0037 0.0011 0.2854 -0.9584>}
 pinky_03_l {parent=pinky_02_l, loc=<2.2046 0.0307 -0.0078>, rot=<-0.0002 0.0098 0.1445 -0.9895>}
 deform_wrist_dn_base_l {parent=hand_l, loc=<0 -1.5326 0>, rot=<0 0 0 -1>}
 deform_wrist_dn_l {parent=deform_wrist_dn_base_l, loc=<0 0 0>, rot=<0 0 0 1>}
 deform_wrist_up_base_l {parent=hand_l, loc=<0 2.3009 0>, rot=<0 0 0 -1>}
 deform_wrist_up_l {parent=deform_wrist_up_base_l, loc=<0 0 0>, rot=<0 0 0 1>}
 lowerarm_twist_02_l {parent=lowerarm_l, loc=<10.6707 0 0>, rot=<0 0 0 1>}
 lowerarm_twist_01_l {parent=lowerarm_l, loc=<18.316 0 0>, rot=<0 0 0 1>}
 deform_elbow_in_base_l {parent=lowerarm_l, loc=<0.1834 3.6062 -0.2572>, rot=<0 0 0 -1>}
 deform_elbow_in_l {parent=deform_elbow_in_base_l, loc=<0 0 0>, rot=<0 0 0 1>}
 deform_elbow_out_base_l {parent=lowerarm_l, loc=<0 -4 0>, rot=<0 0 0 1>}
 deform_elbow_out_l {parent=deform_elbow_out_base_l, loc=<0 0 0>, rot=<0 0 0 1>}
 upperarm_twist_01_l {parent=upperarm_l, loc=<7.2135 -0.1134 0>, rot=<0 0 0 1>}
 base_flap_arm_l {parent=upperarm_twist_01_l, loc=<7.2045 0.8333 3.1027>, rot=<0.1024 0.1189 0.7138 -0.6825>}
 dyn_flap_arm_l {parent=base_flap_arm_l, loc=<0 0 0>, rot=<0 0 0 1>}
 deform_upperarm_bk_base_l {parent=upperarm_twist_01_l, loc=<-6.4053 -4.3746 0.314>, rot=<0 0 0 -1>}
 deform_upperarm_bk_l {parent=deform_upperarm_bk_base_l, loc=<0 0 0>, rot=<0 0 0 1>}
 deform_upperarm_fr_base_l {parent=upperarm_twist_01_l, loc=<-5.6312 5.092 0.3354>, rot=<0 0 0 -1>}
 deform_upperarm_fr_l {parent=deform_upperarm_fr_base_l, loc=<0 0 0>, rot=<0 0 0 1>}
 deform_upperarm_up_base_l {parent=upperarm_twist_01_l, loc=<-8.891 0.4825 3.9517>, rot=<0 0 0 -1>}
 deform_upperarm_up_l {parent=deform_upperarm_up_base_l, loc=<0 0 0>, rot=<0 0 0 1>}
 upperarm_twist_02_l {parent=upperarm_l, loc=<19.7998 -0.1134 0>, rot=<0 0 0 1>}
 clavicle_r {parent=spine_05, loc=<5.4831 -0.1053 3.2946>, rot=<0.7289 0.0465 -0.6829 0.009>}
 upperarm_r {parent=clavicle_r, loc=<-9.8658 1.2136 0.4105>, rot=<0.0164 -0.3245 -0.0269 -0.9454>}
 lowerarm_r {parent=upperarm_r, loc=<-26.465 0.8345 -0.259>, rot=<0.0036 0.0161 -0.2298 -0.9731>}
 hand_r {parent=lowerarm_r, loc=<-22.7355 0.3203 0.6106>, rot=<-0.6784 -0.0223 0.0735 -0.7307>}
 thumb_01_r {parent=hand_r, loc=<-1.3118 0.3191 2.459>, rot=<0.6546 -0.4042 0.0664 -0.6354>}
 thumb_02_r {parent=thumb_01_r, loc=<-5.9336 0.0448 -0.0401>, rot=<-0.0037 0.0213 0.298 -0.9543>}
 thumb_03_r {parent=thumb_02_r, loc=<-2.2223 -0.3682 0.0907>, rot=<0.0118 -0.013 0.0066 -0.9998>}
 index_metacarpal_r {parent=hand_r, loc=<-1.6535 -0.0183 2.96>, rot=<0.0158 -0.0124 -0.0173 -0.9996>}
 index_01_r {parent=index_metacarpal_r, loc=<-8.3008 -0.0357 0.0524>, rot=<-0.0044 0.0485 0.0924 -0.9945>}
 index_02_r {parent=index_01_r, loc=<-3.3587 0.3004 -0.0452>, rot=<-0.002 0.0027 0.3225 -0.9466>}
 index_03_r {parent=index_02_r, loc=<-2.5578 0.123 0.0252>, rot=<-0.002 0.0067 0.2558 -0.9667>}
 middle_metacarpal_r {parent=hand_r, loc=<-2.1856 0.1725 0.5984>, rot=<0.006 0.0168 -0.0329 -0.9993>}
 middle_01_r {parent=middle_metacarpal_r, loc=<-7.3275 0.2447 -0.0129>, rot=<-0.0208 0.0042 0.0858 -0.9961>}
 middle_02_r {parent=middle_01_r, loc=<-4.0123 -0.1401 -0.1175>, rot=<-0.0218 0.0258 0.3512 -0.9357>}
 middle_03_r {parent=middle_02_r, loc=<-2.6077 -0.0534 -0.1373>, rot=<0.0082 0.0188 0.1708 -0.9851>}
 ring_metacarpal_r {parent=hand_r, loc=<-2.2074 0.3625 -1.3638>, rot=<-0.0284 0.0722 -0.0458 -0.9959>}
 ring_01_r {parent=ring_metacarpal_r, loc=<-7.0295 0.6197 0.1446>, rot=<-0.0489 -0.0252 0.1892 -0.9804>}
 ring_02_r {parent=ring_01_r, loc=<-3.6174 -0.4647 0.1027>, rot=<0 0 0.2806 -0.9598>}
 ring_03_r {parent=ring_02_r, loc=<-2.4152 -0.114 -0.0199>, rot=<0 0 0.1189 -0.9929>}
 pinky_metacarpal_r {parent=hand_r, loc=<-1.7616 0.9753 -3.3814>, rot=<-0.0397 0.0814 -0.069 -0.9935>}
 pinky_01_r {parent=pinky_metacarpal_r, loc=<-7.1709 0.4831 0.1094>, rot=<-0.0654 -0.0035 0.2242 -0.9723>}
 pinky_02_r {parent=pinky_01_r, loc=<-2.8274 -0.0893 0.1043>, rot=<0.0037 0.0011 0.2854 -0.9584>}
 pinky_03_r {parent=pinky_02_r, loc=<-2.2046 -0.0307 0.0078>, rot=<-0.0002 0.0098 0.1445 -0.9895>}
 weapon_r {parent=hand_r, loc=<0.1974 3.4111 -0.3811>, rot=<0 0 0 1>}
 hand_attach_r {parent=hand_r, loc=<-8.5001 2.9999 0>, rot=<0 0 -0.7071 -0.7071>}
 deform_wrist_dn_base_r {parent=hand_r, loc=<-0.0001 1.5328 0>, rot=<0 0 0 -1>}
 deform_wrist_dn_r {parent=deform_wrist_dn_base_r, loc=<0 0 0>, rot=<0 0 0 1>}
 deform_wrist_up_base_r {parent=hand_r, loc=<-0.0002 -2.3007 -0.0001>, rot=<0 0 0 -1>}
 deform_wrist_up_r {parent=deform_wrist_up_base_r, loc=<0 0 0>, rot=<0 0 0 1>}
 lowerarm_twist_02_r {parent=lowerarm_r, loc=<-10.6707 0 0>, rot=<0 0 0 -1>}
 lowerarm_twist_01_r {parent=lowerarm_r, loc=<-18.316 0 0>, rot=<0 0 0 -1>}
 deform_elbow_in_base_r {parent=lowerarm_r, loc=<-0.1834 -3.6062 0.2572>, rot=<0 0 0 -1>}
 deform_elbow_in_r {parent=deform_elbow_in_base_r, loc=<0 0 0>, rot=<0 0 0 1>}
 deform_elbow_out_base_r {parent=lowerarm_r, loc=<0 4 -0.0001>, rot=<0 0 0 -1>}
 deform_elbow_out_r {parent=deform_elbow_out_base_r, loc=<0 0 0>, rot=<0 0 0 1>}
 upperarm_twist_01_r {parent=upperarm_r, loc=<-7.2135 0.1134 0>, rot=<0 0 0 -1>}
 deform_upperarm_bk_base_r {parent=upperarm_twist_01_r, loc=<6.4053 4.3746 -0.314>, rot=<0 0 0 -1>}
 deform_upperarm_bk_r {parent=deform_upperarm_bk_base_r, loc=<0 0 0>, rot=<0 0 0 1>}
 deform_upperarm_fr_base_r {parent=upperarm_twist_01_r, loc=<5.6312 -5.092 -0.3354>, rot=<0 0 0 -1>}
 deform_upperarm_fr_r {parent=deform_upperarm_fr_base_r, loc=<0 0 0>, rot=<0 0 0 1>}
 deform_upperarm_up_base_r {parent=upperarm_twist_01_r, loc=<8.891 -0.4825 -3.9517>, rot=<0 0 0 -1>}
 deform_upperarm_up_r {parent=deform_upperarm_up_base_r, loc=<0 0 0>, rot=<0 0 0 1>}
 upperarm_twist_02_r {parent=upperarm_r, loc=<-19.7998 0.1134 0>, rot=<0 0 0 -1>}
 base_flap_down_arm_r {parent=upperarm_twist_02_r, loc=<1.9229 -0.7651 -3.3317>, rot=<0.692 0.673 -0.1535 0.2111>}
 dyn_flap_down_arm_r {parent=base_flap_down_arm_r, loc=<0 0 0>, rot=<0 0 0 1>}
 base_flap_upper_arm_r {parent=upperarm_twist_02_r, loc=<5.6109 -1.0848 -3.2201>, rot=<0.6827 0.6792 -0.1521 0.2226>}
 dyn_flap_upper_arm_r {parent=base_flap_upper_arm_r, loc=<0 0 0>, rot=<0 0 0 1>}
 attach_backpack {parent=spine_05, loc=<-0.123 -13.473 0>, rot=<-0.0093 -0.707 0.0093 -0.707>}
 attach_cape {parent=spine_05, loc=<9.017 -11.714 0>, rot=<0.0616 -0.7044 -0.0616 -0.7044>}
 base_necklace_01_cr {parent=spine_05, loc=<11.3138 0.1529 0>, rot=<0 0 0.8779 0.4788>}
 dyn_necklace_01_cr {parent=base_necklace_01_cr, loc=<0 0 0>, rot=<0 0 0 1>}
 dyn_necklace_02_cr {parent=dyn_necklace_01_cr, loc=<6.1652 1.138 -0.001>, rot=<0 0 -0.116 -0.9933>}
 dyn_necklace_03_cr {parent=dyn_necklace_02_cr, loc=<3.7486 0.4378 0>, rot=<0 0 -0.0571 -0.9984>}
 aim_necklace_cr {parent=spine_05, loc=<4.0089 7.0077 -0.001>, rot=<0 0 0.9473 0.3202>}
 base_hair_ctrl {controlType=EULER_TRANSFORM, parent=spine_05, type=Control, loc=<0 0 0>, rot=<0 0 0 1>}
 base_jacket {parent=spine_04, loc=<0 0 0>, rot=<0 0 0 1>}
 base_jacket_r_2 {parent=base_jacket, loc=<2.1249 -7.22 8.9099>, rot=<0.0208 -0.4996 0.8653 -0.036>}
 dyn_jacket_r_2 {parent=base_jacket_r_2, loc=<0 0 0>, rot=<0 0 0 1>}
 base_jacket_r_1 {parent=base_jacket, loc=<1.32 2.4306 12.2275>, rot=<0 -0.0004 0.9991 -0.0416>}
 dyn_jacket_r_1 {parent=base_jacket_r_1, loc=<0 0 0>, rot=<0 0 0 1>}
 base_jacket_r_0 {parent=base_jacket, loc=<0.6278 10.7305 9.0988>, rot=<-0.0188 0.4518 0.8911 -0.0371>}
 dyn_jacket_r_0 {parent=base_jacket_r_0, loc=<0 0 0>, rot=<0 0 0 1>}
 base_jacket_back {parent=base_jacket, loc=<2.1053 -7.5413 0>, rot=<0.0294 -0.7065 0.7065 -0.0294>}
 dyn_jacket_back {parent=base_jacket_back, loc=<0 0 0>, rot=<0 0 0 1>}
 base_jacket_l_2 {parent=base_jacket, loc=<2.1247 -7.22 -8.9099>, rot=<0.036 -0.8653 0.4996 -0.0208>}
 dyn_jacket_l_2 {parent=base_jacket_l_2, loc=<0 0 0>, rot=<0 0 0 1>}
 base_jacket_l_1 {parent=base_jacket, loc=<1.3198 2.4306 -12.2275>, rot=<0.0416 -0.9991 0.0004 0>}
 dyn_jacket_l_1 {parent=base_jacket_l_1, loc=<0 0 0>, rot=<0 0 0 1>}
 base_jacket_l_0 {parent=base_jacket, loc=<0.6276 10.7304 -9.0988>, rot=<0.0371 -0.8911 -0.4518 0.0188>}
 dyn_jacket_l_0 {parent=base_jacket_l_0, loc=<0 0 0>, rot=<0 0 0 1>}
 deform_pec_aim_base_l {parent=spine_04, loc=<3.2456 9.3309 -1.2237>, rot=<0.0603 -0.4033 0.3318 -0.8507>}
 deform_pec_aim_l {parent=deform_pec_aim_base_l, loc=<0 0 0>, rot=<0 0 0 1>}
 deform_pec_base_l {parent=deform_pec_aim_l, loc=<9.3665 0 0>, rot=<0 0 0 1>}
 deform_pec_l {parent=deform_pec_base_l, loc=<0 0 0>, rot=<0 0 0 1>}
 deform_pec_aim_base_r {parent=spine_04, loc=<3.2452 9.3309 1.2237>, rot=<-0.4033 -0.0603 0.8507 0.3318>}
 deform_pec_aim_r {parent=deform_pec_aim_base_r, loc=<0 0 0>, rot=<0 0 0 1>}
 deform_pec_base_r {parent=deform_pec_aim_r, loc=<-9.3667 -0.0003 -0.0004>, rot=<0 0 0 1>}
 deform_pec_r {parent=deform_pec_base_r, loc=<0 0 0>, rot=<0 0 0 1>}
 thigh_l {parent=pelvis, loc=<-1.0855 -0.2109 -8.7053>, rot=<0.0009 0.0355 0.0155 -0.9993>}
 calf_l {parent=thigh_l, loc=<-42.5284 -0.0019 -0.7297>, rot=<0.0006 -0.0161 -0.0391 -0.9991>}
 foot_l {parent=calf_l, loc=<-46.7401 -2.5014 -0.592>, rot=<0.0427 -0.0238 0.0353 -0.9982>}
 ball_l {parent=foot_l, loc=<-3.5658 10.8066 -0.1544>, rot=<0.0001 -0.0001 -0.7063 -0.7079>}
 calf_twist_02_l {parent=calf_l, loc=<-19.1546 -1.7943 -0.2638>, rot=<0.0237 0.0095 -0.0035 -0.9997>}
 calf_twist_01_l {parent=calf_l, loc=<-35.8483 -1.9197 -0.5794>, rot=<0.0237 0.0095 -0.0035 -0.9997>}
 deform_knee_in_base_l {parent=calf_l, loc=<0.8031 -5.0231 0.2456>, rot=<0 0 0 -1>}
 deform_knee_in_l {parent=deform_knee_in_base_l, loc=<0 0 0>, rot=<0 0 0 1>}
 deform_knee_out_base_l {parent=calf_l, loc=<0 4 0>, rot=<0 0 0 -1>}
 deform_knee_out_l {parent=deform_knee_out_base_l, loc=<0 0 0>, rot=<0 0 0 1>}
 thigh_twist_01_l {parent=thigh_l, loc=<-27.952 0 0>, rot=<0.0013 0 0 -1>}
 base_flap_down_boot_l {parent=thigh_twist_01_l, loc=<-9.5595 0.7333 -5.1701>, rot=<0.6448 -0.6975 -0.284 -0.1305>}
 dyn_flap_down_boot_l {parent=base_flap_down_boot_l, loc=<0 0 0>, rot=<0 0 0 1>}
 base_flap_upper_boot_l {parent=thigh_twist_01_l, loc=<-5.5499 1.2523 -5.8221>, rot=<0.6456 -0.7055 -0.2779 -0.0912>}
 dyn_flap_upper_boot_l {parent=base_flap_upper_boot_l, loc=<0 0 0>, rot=<0 0 0 1>}
 deform_groin_base_l {parent=thigh_l, loc=<0.4801 5.6776 -0.2486>, rot=<0 0 0 -1>}
 deform_groin_l {parent=deform_groin_base_l, loc=<0 0 0>, rot=<0 0 0 1>}
 base_helper_winterstart_l {parent=thigh_l, loc=<3.7708 6.3636 -0.3945>, rot=<-0.4372 -0.6446 -0.2873 -0.5574>}
 helper_winterstart_l {parent=base_helper_winterstart_l, loc=<0 0 0>, rot=<0 0 0 1>}
 thigh_r {parent=pelvis, loc=<-1.0855 -0.2109 8.7053>, rot=<0.0355 -0.0009 0.9993 0.0155>}
 calf_r {parent=thigh_r, loc=<42.5284 0.0019 0.7297>, rot=<0.0006 -0.0161 -0.0391 -0.9991>}
 foot_r {parent=calf_r, loc=<46.7401 2.5014 0.592>, rot=<0.0427 -0.0238 0.0353 -0.9982>}
 ball_r {parent=foot_r, loc=<3.5658 -10.8066 0.1544>, rot=<0.0001 -0.0001 -0.7063 -0.7079>}
 calf_twist_02_r {parent=calf_r, loc=<19.1546 1.7943 0.2638>, rot=<0.0237 0.0095 -0.0035 -0.9997>}
 calf_twist_01_r {parent=calf_r, loc=<35.8483 1.9197 0.5794>, rot=<0.0237 0.0095 -0.0035 -0.9997>}
 deform_knee_in_base_r {parent=calf_r, loc=<-0.8031 5.0231 -0.2457>, rot=<0 0 0 -1>}
 deform_knee_in_r {parent=deform_knee_in_base_r, loc=<0 0 0>, rot=<0 0 0 1>}
 deform_knee_out_base_r {parent=calf_r, loc=<0 -4 0>, rot=<0 0 0 -1>}
 deform_knee_out_r {parent=deform_knee_out_base_r, loc=<0 0 0>, rot=<0 0 0 1>}
 thigh_twist_01_r {parent=thigh_r, loc=<27.952 0 0>, rot=<0.0013 0 0 -1>}
 base_flap_boot_r {parent=thigh_twist_01_r, loc=<7.6171 -1.1205 5.6655>, rot=<-0.1919 -0.0072 -0.6938 -0.6941>}
 dyn_flap_boot_r {parent=base_flap_boot_r, loc=<0 0 0>, rot=<0 0 0 1>}
 deform_groin_base_r {parent=thigh_r, loc=<-0.4801 -5.6776 0.2486>, rot=<0 0 0 -1>}
 deform_groin_r {parent=deform_groin_base_r, loc=<0 0 0>, rot=<0 0 0 1>}
 deform_glute_base_l {parent=pelvis, loc=<-7.3466 -9.0213 -9.1245>, rot=<0.0172 0.031 -0.4477 -0.8935>}
 deform_glute_l {parent=deform_glute_base_l, loc=<0 0 0>, rot=<0 0 0 1>}
 deform_glute_base_r {parent=pelvis, loc=<-7.3464 -9.0213 9.1245>, rot=<0.031 -0.0172 0.8935 -0.4477>}
 deform_glute_r {parent=deform_glute_base_r, loc=<0 0 0>, rot=<0 0 0 1>}
 attach {parent=root, loc=<0 0 0>, rot=<0 0 0 1>}
 ik_hand_root {parent=root, loc=<0 0 0>, rot=<0 0 0 1>}
 ik_hand_gun {parent=ik_hand_root, loc=<-48.1743 3.3545 104.4381>, rot=<0.6276 0.1309 0.4183 -0.6434>}
 ik_hand_r {parent=ik_hand_gun, loc=<0 0 0>, rot=<0 0 0 1>}
 ik_hand_l {parent=ik_hand_gun, loc=<59.3271 67.6997 34.3524>, rot=<-0.6158 -0.7027 -0.3565 0>}
 ik_foot_root {parent=root, loc=<0 0 0>, rot=<0 0 0 1>}
 ik_foot_l {parent=ik_foot_root, loc=<14.8262 -7.3545 8.296>, rot=<0.0357 0.7045 0.0284 -0.7082>}
 ik_foot_r {parent=ik_foot_root, loc=<-14.8262 -7.3545 8.296>, rot=<0.7082 0.0284 -0.7045 0.0357>}
 root_fp {parent=root, loc=<0 0 0>, rot=<0 0 0 1>}
 attach_fp {parent=root_fp, loc=<0 0 0>, rot=<0 0 0 1>}
 camera_root {parent=root_fp, loc=<0 0 0>, rot=<0 0 0 1>}
 camera {parent=camera_root, loc=<0 0 0>, rot=<0 0 0 1>}
 attach_emote_01 {parent=root, loc=<0 0 0>, rot=<0 0 0 1>}
 attach_emote_02 {parent=root, loc=<0 0 0>, rot=<0 0 0 1>}
 attach_emote_03 {parent=root, loc=<0 0 0>, rot=<0 0 0 1>}
 warp_point_01 {parent=root, loc=<0 0 0>, rot=<0 0 0 1>}
 warp_point_02 {parent=root, loc=<0 0 0>, rot=<0 0 0 1>}
 warp_point_03 {parent=root, loc=<0 0 0>, rot=<0 0 0 1>}
]
variables [
 headSwing_rot {type=FVector, default=(X=0,Y=0,Z=0)}
 headSwing_twist {type=FVector, default=(X=0,Y=0,Z=0)}
]
