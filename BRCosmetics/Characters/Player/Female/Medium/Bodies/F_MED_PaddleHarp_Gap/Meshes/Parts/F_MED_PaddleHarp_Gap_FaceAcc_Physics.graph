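# Physics asset for the face accessory part of the female medium PaddleHarp_Gap player cosmetic, providing physics for its right-side hair bones.
# Back, side and front-side chains of three capsules, plus a single front link, hang from small base capsules constrained to the head.
# Hair joints use narrow swing and zero twist limits.
# Larger head, neck, upper spine, right clavicle and upper arm capsules approximate the body, joined by free angular constraints.
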PhysicsAsset /BRCosmetics/Characters/Player/Female/Medium/Bodies/F_MED_PaddleHarp_Gap/Meshes/Parts/F_MED_PaddleHarp_Gap_FaceAcc_Physics
bodies {
 base_hair_back_r [{shapeType=Capsule, length=0.505, radius=0.505, shapeName=base_front_r_capsule, rotation=<rot 0 -90 0>}]
 base_hair_font_side_r [{shapeType=Capsule, length=0.505, radius=0.505, shapeName=base_front_r_capsule, rotation=<rot 0 -90 0>}]
 base_hair_front_01_r [{shapeType=Capsule, length=0.505, radius=0.505, shapeName=base_front_r_capsule, rotation=<rot 0 -90 0>}]
 base_hair_side_r [{shapeType=Capsule, length=0.505, radius=0.505, shapeName=base_front_r_capsule, rotation=<rot 0 -90 0>}]
 clavicle_r [{
   shapeType=Capsule
   length=5.8741
   radius=7.3343
   shapeName=clavicle_r_capsule
   center=<-4.5939 1.0938 1.0166>
   rotation=<rot 83.3961 -0.001 179.999>
  }]
 dyn_hair_back_r_01 [{shapeType=Capsule, length=3, radius=1, shapeName=dyn_front_r_01_capsule, center=<1 0 0>, rotation=<rot 90 0 0>}]
 dyn_hair_back_r_02 [{shapeType=Capsule, length=3, radius=1, shapeName=dyn_front_r_01_capsule, center=<1 0 0>, rotation=<rot 90 0 0>}]
 dyn_hair_back_r_03 [{shapeType=Capsule, length=3, radius=1, shapeName=dyn_front_r_01_capsule, center=<1 0 0>, rotation=<rot 90 0 0>}]
 dyn_hair_front_01_r [{shapeType=Capsule, length=3, radius=1, shapeName=dyn_front_r_01_capsule, center=<2 0 0>, rotation=<rot 90 0 0>}]
 dyn_hair_front_side_r_01 [{shapeType=Capsule, length=3, radius=1, shapeName=dyn_front_r_01_capsule, center=<1 0 0>, rotation=<rot 90 0 0>}]
 dyn_hair_front_side_r_02 [{shapeType=Capsule, length=3, radius=1, shapeName=dyn_front_r_01_capsule, center=<1 0 0>, rotation=<rot 90 0 0>}]
 dyn_hair_front_side_r_03 [{shapeType=Capsule, length=3, radius=1, shapeName=dyn_front_r_01_capsule, center=<1 0 0>, rotation=<rot 90 0 0>}]
 dyn_hair_side_r_01 [{shapeType=Capsule, length=3, radius=1, shapeName=dyn_front_r_01_capsule, center=<1 0 0>, rotation=<rot 90 0 0>}]
 dyn_hair_side_r_02 [{shapeType=Capsule, length=3, radius=1, shapeName=dyn_front_r_01_capsule, center=<1 0 0>, rotation=<rot 90 0 0>}]
 dyn_hair_side_r_03 [{shapeType=Capsule, length=3, radius=1, shapeName=dyn_front_r_01_capsule, center=<1 0 0>, rotation=<rot 90 0 0>}]
 head [{shapeType=Capsule, length=10, radius=8, shapeName=head_capsule, center=<0.7982 0.8112 0>, rotation=<rot 90 0 0>}]
 neck_01 [{shapeType=Capsule, length=0.5295, radius=4.6745, shapeName=neck_01_capsule, center=<0 0.4185 0>}]
 spine_05 [
  {shapeType=Capsule, length=0, radius=9.1526, shapeName=spine_05_capsule, center=<3.455 0.469 0>, rotation=<rot 0 -90 0>}
  {shapeType=Capsule, length=7.3718, radius=9.8741, shapeName=spine_05_capsule, center=<-1.1311 1.6378 0>}
 ]
 upperarm_r [{
   shapeType=Capsule
   length=18.4782
   radius=6.5
   shapeName=upperarm_r_capsule
   center=<-8.7766 1.0401 0.805>
   rotation=<rot 89.8676 172.1179 -6.3248>
  }]
}
constraints [
 {
  bone1Name=head
  bone2Name=neck_01
  swing1LimitDegrees=45
  swing1Motion=Free
  swing2LimitDegrees=45
  swing2Motion=Free
  twistLimitDegrees=45
  twistMotion=Free
 }
 {
  bone1Name=neck_01
  bone2Name=spine_05
  swing1LimitDegrees=45
  swing1Motion=Free
  swing2LimitDegrees=45
  swing2Motion=Free
  twistLimitDegrees=45
  twistMotion=Free
 }
 {
  bone1Name=base_hair_side_r
  bone2Name=head
  swing1LimitDegrees=45
  swing1Motion=Free
  swing2LimitDegrees=45
  swing2Motion=Free
  twistLimitDegrees=45
  twistMotion=Free
 }
 {bone1Name=dyn_hair_side_r_01, bone2Name=base_hair_side_r, swing1LimitDegrees=15, swing2LimitDegrees=15, twistLimitDegrees=0}
 {bone1Name=dyn_hair_side_r_02, bone2Name=dyn_hair_side_r_01, swing1LimitDegrees=15, swing2LimitDegrees=15, twistLimitDegrees=0}
 {bone1Name=dyn_hair_side_r_03, bone2Name=dyn_hair_side_r_02, swing1LimitDegrees=15, swing2LimitDegrees=15, twistLimitDegrees=0}
 {bone1Name=dyn_hair_front_01_r, bone2Name=base_hair_front_01_r, swing1LimitDegrees=20, swing2LimitDegrees=15, twistLimitDegrees=0}
 {
  bone1Name=base_hair_front_01_r
  bone2Name=head
  swing1LimitDegrees=45
  swing1Motion=Free
  swing2LimitDegrees=45
  swing2Motion=Free
  twistLimitDegrees=45
  twistMotion=Free
 }
 {bone1Name=dyn_hair_front_side_r_03, bone2Name=dyn_hair_front_side_r_02, swing1LimitDegrees=15, swing2LimitDegrees=3, twistLimitDegrees=0}
 {bone1Name=dyn_hair_front_side_r_02, bone2Name=dyn_hair_front_side_r_01, swing1LimitDegrees=15, swing2LimitDegrees=3, twistLimitDegrees=0}
 {bone1Name=dyn_hair_front_side_r_01, bone2Name=base_hair_font_side_r, swing1LimitDegrees=15, swing2LimitDegrees=5, twistLimitDegrees=0}
 {
  bone1Name=base_hair_font_side_r
  bone2Name=head
  swing1LimitDegrees=45
  swing1Motion=Free
  swing2LimitDegrees=45
  swing2Motion=Free
  twistLimitDegrees=45
  twistMotion=Free
 }
 {bone1Name=dyn_hair_back_r_03, bone2Name=dyn_hair_back_r_02, swing1LimitDegrees=8, swing2LimitDegrees=10, twistLimitDegrees=0}
 {bone1Name=dyn_hair_back_r_02, bone2Name=dyn_hair_back_r_01, swing1LimitDegrees=8, swing2LimitDegrees=10, twistLimitDegrees=0}
 {bone1Name=dyn_hair_back_r_01, bone2Name=base_hair_back_r, swing1LimitDegrees=8, swing2LimitDegrees=10, twistLimitDegrees=0}
 {
  bone1Name=base_hair_back_r
  bone2Name=head
  swing1LimitDegrees=45
  swing1Motion=Free
  swing2LimitDegrees=45
  swing2Motion=Free
  twistLimitDegrees=45
  twistMotion=Free
 }
 {
  bone1Name=clavicle_r
  bone2Name=spine_05
  swing1LimitDegrees=45
  swing1Motion=Free
  swing2LimitDegrees=45
  swing2Motion=Free
  twistLimitDegrees=45
  twistMotion=Free
 }
 {
  bone1Name=upperarm_r
  bone2Name=clavicle_r
  swing1LimitDegrees=45
  swing1Motion=Free
  swing2LimitDegrees=45
  swing2Motion=Free
  twistLimitDegrees=45
  twistMotion=Free
 }
]
