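# Post-process-style AnimBP for the PaddleHarp Gap female face accessory, adding secondary motion to hair and earrings.
# Blends poses copied from parent, body and head meshes, then runs a Control Rig and RigidBody simulation.
# Constrains earring bases to the head before chained, spring-based AnimDynamics nodes simulate earrings and hair bones.
# The event graph caches part meshes and updates gravity override (using SpadeStrut parameters), head-up alpha and RbanAlpha, lowered in frontend.
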AnimBlueprint /BRCosmetics/Characters/Player/Female/Medium/Bodies/F_MED_PaddleHarp_Gap/Meshes/Parts/F_MED_PaddleHarp_Gap_FaceAcc_AnimBP
target_skeleton=&/F_MED_PaddleHarp_Gap_FaceAcc_Skeleton
graphs [
 AnimGraph {nodes [
   Output Pose(Result=Use cached pose 'dyn_cache'()) {layerGroup=None}
   n1 = Control Rig(Source=Use cached pose 'Start Cache'()) {controlRigAssetReference {blueprintRigClass=&/F_MED_PaddleHarp_Gap_FaceAcc_CtrlRig.F_MED_PaddleHarp_Gap_FaceAcc_CtrlRig_C}}
   n2 = RigidBody!(ComponentPose=Local To Component(LocalPose=@n1), Alpha, OverrideWorldGravity:Vector="0,0,0") -> (Pose) {
    bOverrideWorldGravity=true
    overridePhysicsAsset=&/F_MED_PaddleHarp_Gap_FaceAcc_Physics
    simulationSpace=BaseBoneSpace
    baseBoneRef {boneName=spine_05}
    simSpaceSettings {
     dampingAlpha=0
     maxAngularAcceleration=16
     maxAngularVelocity=4
     maxLinearAcceleration=800
     maxLinearVelocity=200
     velocityScaleZ=0.5
     worldAlpha=0.5
     externalAngularVelocity=<0.2 0.2 0.2>
     externalLinearVelocity=<0.2 0.2 0.2>
    }
   }
   Save cached pose 'control_rig_cache'(Pose=Component To Local(ComponentPose=@n2)) {cacheName=control_rig_cache}
   n3 = "Get Face Mesh "() -> ("Face Mesh ":Skeletal Mesh Component Object Reference)
   n4 = Copy Pose From Mesh(SourceMeshComponent="Get Body Mesh "() -> ("Body Mesh ":Skeletal Mesh Component Object Reference)) {bUseAttachedParent=true}
   n5 = Copy Pose From Mesh() {bUseAttachedParent=true}
   Save cached pose 'Start Cache'(Pose=Layered blend per bone(
     BasePose=Layered blend per bone(BasePose=@n5, BlendPoses_0=@n4) {layerSetup [{branchFilters [{boneName=clavicle_r}]}]}
     BlendPoses_0=Copy Pose From Mesh(SourceMeshComponent=@n3) {bUseAttachedParent=true}
    ) {layerSetup [{branchFilters [{boneName=FACIAL_R_ForeheadMid}]}]}) {cacheName=Start Cache}
   n6 = Anim Dynamics - Bone: dyn_earring_l(ComponentPose=Local To Component(LocalPose=Use cached pose 'predyn_cache'()), Alpha=1, ExternalForce="0,0,0") {
    angularDampingOverride=1
    angularSpringConstant=1500
    bAngularSpring=true
    bOverrideAngularDamping=true
    bUseGravityOverride=true
    gravityScale=0.35
    simulationSpace=RootRelative
    boundBone {boneName=dyn_earring_l}
    componentLinearAccScale=<1 1 1>
    componentLinearVelScale=<1 1 1>
    gravityOverride {z=-980}
    physicsBodyDefinitions [{
      boundBone {boneName=dyn_earring_l}
      boxExtents=<20 30 30>
      constraintSetup {angularLimitsMax=<0 120 30>, angularLimitsMin=<0 -5 -40>, angularTarget=<1 0 0>}
      localJointOffset=<4 0 0>
     }]
    simSpaceSettings {maxAngularAcceleration=35, maxAngularVelocity=7, simSpaceAngularAlpha=1}
   }
   n7 = Anim Dynamics - Bone: dyn_earring_r(ComponentPose=@n6, Alpha=1, ExternalForce="0,0,0") {
    angularDampingOverride=1
    angularSpringConstant=1500
    bAngularSpring=true
    bOverrideAngularDamping=true
    bUseGravityOverride=true
    gravityScale=0.35
    simulationSpace=RootRelative
    boundBone {boneName=dyn_earring_r}
    componentLinearAccScale=<1 1 1>
    componentLinearVelScale=<1 1 1>
    gravityOverride {z=-980}
    physicsBodyDefinitions [{
      boundBone {boneName=dyn_earring_r}
      boxExtents=<20 30 30>
      constraintSetup {angularLimitsMax=<0 120 30>, angularLimitsMin=<0 -5 -60>, angularTarget=<1 0 0>}
      localJointOffset=<4 0 0>
     }]
    simSpaceSettings {maxAngularAcceleration=35, maxAngularVelocity=7, simSpaceAngularAlpha=1}
   }
   n8 = Anim Dynamics - Bone: dyn_hair_add_01(ComponentPose=@n7, Alpha=1, ExternalForce="0,0,0") {
    angularDampingOverride=1
    angularSpringConstant=2000
    bAngularSpring=true
    bOverrideAngularDamping=true
    bUseGravityOverride=true
    gravityScale=0.35
    simulationSpace=RootRelative
    alphaBoolBlend {blendOption=Linear}
    boundBone {boneName=dyn_hair_add_01}
    componentAppliedLinearAccClamp=<3000 3000 3000>
    componentLinearAccScale=<0.6 0.5 0.5>
    componentLinearVelScale=<0.6 0.5 0.5>
    gravityOverride {z=-980}
    physicsBodyDefinitions [{
      boundBone {boneName=dyn_hair_add_01}
      boxExtents=<10 10 10>
      constraintSetup {angularLimitsMax=<20 5 20>, angularLimitsMin=<-15 -50 0>, angularTarget=<1 0 0>}
      localJointOffset=<3 0 0>
     }]
    simSpaceSettings {simSpaceAngularAlpha=1}
   }
   n9 = Anim Dynamics - Bone: dyn_hair_add_02(ComponentPose=@n8, Alpha=1, ExternalForce="0,0,0") {
    angularDampingOverride=1
    angularSpringConstant=2000
    bAngularSpring=true
    bOverrideAngularDamping=true
    bUseGravityOverride=true
    gravityScale=0.35
    simulationSpace=RootRelative
    alphaBoolBlend {blendOption=Linear}
    boundBone {boneName=dyn_hair_add_02}
    componentAppliedLinearAccClamp=<3000 3000 3000>
    componentLinearAccScale=<0.6 0.5 0.5>
    componentLinearVelScale=<0.6 0.5 0.5>
    gravityOverride {z=-980}
    physicsBodyDefinitions [{
      boundBone {boneName=dyn_hair_add_02}
      boxExtents=<10 10 10>
      constraintSetup {angularLimitsMax=<20 5 20>, angularLimitsMin=<-15 -50 0>, angularTarget=<1 0 0>}
      localJointOffset=<3 0 0>
     }]
    simSpaceSettings {simSpaceAngularAlpha=1}
   }
   n10 = Anim Dynamics - Bone: dyn_hair_add_03(ComponentPose=@n9, Alpha=1, ExternalForce="0,0,0") {
    angularDampingOverride=1
    angularSpringConstant=2000
    bAngularSpring=true
    bOverrideAngularDamping=true
    bUseGravityOverride=true
    gravityScale=0.35
    simulationSpace=RootRelative
    alphaBoolBlend {blendOption=Linear}
    boundBone {boneName=dyn_hair_add_03}
    componentAppliedLinearAccClamp=<3000 3000 3000>
    componentLinearAccScale=<0.6 0.5 0.5>
    componentLinearVelScale=<0.6 0.5 0.5>
    gravityOverride {z=-980}
    physicsBodyDefinitions [{
      boundBone {boneName=dyn_hair_add_03}
      boxExtents=<10 10 10>
      constraintSetup {angularLimitsMax=<10 5 18>, angularLimitsMin=<-15 -30 0>, angularTarget=<1 0 0>}
      localJointOffset=<3 0 0>
     }]
    simSpaceSettings {simSpaceAngularAlpha=1}
   }
   n11 = Anim Dynamics - Bone: dyn_hair_add_04(ComponentPose=@n10, Alpha=1, ExternalForce="0,0,0") {
    angularDampingOverride=1
    angularSpringConstant=2000
    bAngularSpring=true
    bOverrideAngularDamping=true
    bUseGravityOverride=true
    gravityScale=0.35
    simulationSpace=RootRelative
    alphaBoolBlend {blendOption=Linear}
    boundBone {boneName=dyn_hair_add_04}
    componentAppliedLinearAccClamp=<3000 3000 3000>
    componentLinearAccScale=<0.6 0.5 0.5>
    componentLinearVelScale=<0.6 0.5 0.5>
    gravityOverride {z=-980}
    physicsBodyDefinitions [{
      boundBone {boneName=dyn_hair_add_04}
      boxExtents=<20 30 30>
      constraintSetup {angularLimitsMax=<10 10 3>, angularLimitsMin=<-10 -10 -10>, angularTarget=<1 0 0>}
      localJointOffset=<4 0 0>
     }]
    simSpaceSettings {simSpaceAngularAlpha=1}
   }
   n12 = Anim Dynamics - Bone: dyn_hair_add_05(ComponentPose=@n11, Alpha=1, ExternalForce="0,0,0") {
    angularDampingOverride=1
    angularSpringConstant=2000
    bAngularSpring=true
    bOverrideAngularDamping=true
    bUseGravityOverride=true
    gravityScale=0.35
    simulationSpace=RootRelative
    alphaBoolBlend {blendOption=Linear}
    boundBone {boneName=dyn_hair_add_05}
    componentAppliedLinearAccClamp=<3000 3000 3000>
    componentLinearAccScale=<0.6 0.5 0.5>
    componentLinearVelScale=<0.6 0.5 0.5>
    gravityOverride {z=-980}
    physicsBodyDefinitions [{
      boundBone {boneName=dyn_hair_add_05}
      boxExtents=<25 30 30>
      constraintSetup {angularLimitsMax=<15 5 20>, angularLimitsMin=<-15 -20 0>, angularTarget=<1 0 0>}
      localJointOffset=<1 0 0>
     }]
    simSpaceSettings {simSpaceAngularAlpha=1}
   }
   n13 = Anim Dynamics - Bone: dyn_hair_back_01(ComponentPose=@n12, Alpha=1, ExternalForce="0,0,0") {
    angularDampingOverride=1
    angularSpringConstant=2000
    bAngularSpring=true
    bOverrideAngularDamping=true
    bUseGravityOverride=true
    gravityScale=0.35
    simulationSpace=RootRelative
    alphaBoolBlend {blendOption=Linear}
    boundBone {boneName=dyn_hair_back_01}
    componentAppliedLinearAccClamp=<3000 3000 3000>
    componentLinearAccScale=<0.6 0.5 0.5>
    componentLinearVelScale=<0.6 0.5 0.5>
    gravityOverride {z=-980}
    physicsBodyDefinitions [{
      boundBone {boneName=dyn_hair_back_01}
      boxExtents=<10 10 10>
      constraintSetup {angularLimitsMax=<15 5 15>, angularLimitsMin=<-10 -40 0>, angularTarget=<1 0 0>}
      localJointOffset=<3 0 0>
     }]
    simSpaceSettings {simSpaceAngularAlpha=1}
   }
   n14 = Anim Dynamics - Bone: base_hair_back_02(ComponentPose=@n13, Alpha=1, ExternalForce="0,0,0") {
    angularDampingOverride=1
    angularSpringConstant=2000
    bAngularSpring=true
    bOverrideAngularDamping=true
    bUseGravityOverride=true
    gravityScale=0.35
    simulationSpace=RootRelative
    alphaBoolBlend {blendOption=Linear}
    boundBone {boneName=base_hair_back_02}
    componentAppliedLinearAccClamp=<3000 3000 3000>
    componentLinearAccScale=<0.6 0.5 0.5>
    componentLinearVelScale=<0.6 0.5 0.5>
    gravityOverride {z=-980}
    physicsBodyDefinitions [{
      boundBone {boneName=base_hair_back_02}
      boxExtents=<10 10 10>
      constraintSetup {angularLimitsMax=<20 5 20>, angularLimitsMin=<-15 -50 0>, angularTarget=<1 0 0>}
      localJointOffset=<3 0 0>
     }]
    simSpaceSettings {simSpaceAngularAlpha=1}
   }
   n15 = Anim Dynamics - Bone: dyn_hair_back_03(ComponentPose=@n14, Alpha=1, ExternalForce="0,0,0") {
    angularDampingOverride=1
    angularSpringConstant=2000
    bAngularSpring=true
    bOverrideAngularDamping=true
    bUseGravityOverride=true
    gravityScale=0.35
    simulationSpace=RootRelative
    alphaBoolBlend {blendOption=Linear}
    boundBone {boneName=dyn_hair_back_03}
    componentAppliedLinearAccClamp=<3000 3000 3000>
    componentLinearAccScale=<0.6 0.5 0.5>
    componentLinearVelScale=<0.6 0.5 0.5>
    gravityOverride {z=-980}
    physicsBodyDefinitions [{
      boundBone {boneName=dyn_hair_back_03}
      boxExtents=<10 10 10>
      constraintSetup {angularLimitsMax=<18 5 18>, angularLimitsMin=<-18 -45 0>, angularTarget=<1 0 0>}
      localJointOffset=<3 0 0>
     }]
    simSpaceSettings {simSpaceAngularAlpha=1}
   }
   n16 = Anim Dynamics - Bone: dyn_hair_back_04(ComponentPose=@n15, Alpha=1, ExternalForce="0,0,0") {
    angularDampingOverride=1
    angularSpringConstant=2000
    bAngularSpring=true
    bOverrideAngularDamping=true
    bUseGravityOverride=true
    gravityScale=0.35
    simulationSpace=RootRelative
    alphaBoolBlend {blendOption=Linear}
    boundBone {boneName=dyn_hair_back_04}
    componentAppliedLinearAccClamp=<3000 3000 3000>
    componentLinearAccScale=<0.6 0.5 0.5>
    componentLinearVelScale=<0.6 0.5 0.5>
    gravityOverride {z=-980}
    physicsBodyDefinitions [{
      boundBone {boneName=dyn_hair_back_04}
      boxExtents=<10 10 10>
      constraintSetup {angularLimitsMax=<20 5 20>, angularLimitsMin=<-15 -50 0>, angularTarget=<1 0 0>}
      localJointOffset=<3 0 0>
     }]
    simSpaceSettings {simSpaceAngularAlpha=1}
   }
   n17 = Anim Dynamics - Bone: dyn_hair_back_05(ComponentPose=@n16, Alpha=1, ExternalForce="0,0,0") {
    angularDampingOverride=1
    angularSpringConstant=2000
    bAngularSpring=true
    bOverrideAngularDamping=true
    bUseGravityOverride=true
    gravityScale=0.35
    simulationSpace=RootRelative
    alphaBoolBlend {blendOption=Linear}
    boundBone {boneName=dyn_hair_back_05}
    componentAppliedLinearAccClamp=<3000 3000 3000>
    componentLinearAccScale=<0.6 0.5 0.5>
    componentLinearVelScale=<0.6 0.5 0.5>
    gravityOverride {z=-980}
    physicsBodyDefinitions [{
      boundBone {boneName=dyn_hair_back_05}
      boxExtents=<10 10 10>
      constraintSetup {angularLimitsMax=<18 5 18>, angularLimitsMin=<-18 -45 0>, angularTarget=<1 0 0>}
      localJointOffset=<3 0 0>
     }]
    simSpaceSettings {simSpaceAngularAlpha=1}
   }
   n18 = Anim Dynamics - Bone: dyn_hair_back_06(ComponentPose=@n17, Alpha=1, ExternalForce="0,0,0") {
    angularDampingOverride=1
    angularSpringConstant=2000
    bAngularSpring=true
    bOverrideAngularDamping=true
    bUseGravityOverride=true
    gravityScale=0.35
    simulationSpace=RootRelative
    alphaBoolBlend {blendOption=Linear}
    boundBone {boneName=dyn_hair_back_06}
    componentAppliedLinearAccClamp=<3000 3000 3000>
    componentLinearAccScale=<0.6 0.5 0.5>
    componentLinearVelScale=<0.6 0.5 0.5>
    gravityOverride {z=-980}
    physicsBodyDefinitions [{
      boundBone {boneName=dyn_hair_back_06}
      boxExtents=<10 10 10>
      constraintSetup {angularLimitsMax=<15 5 15>, angularLimitsMin=<-12 -10 0>, angularTarget=<1 0 0>}
      localJointOffset=<3 0 0>
     }]
    simSpaceSettings {simSpaceAngularAlpha=1}
   }
   Save cached pose 'dyn_cache'(Pose=Component To Local(ComponentPose=@n18)) {cacheName=dyn_cache}
   n19 = "Constraint - base_earring_l (P)"!(
    ComponentPose=Local To Component(LocalPose=Use cached pose 'control_rig_cache'())
    Alpha=1
    ConstraintWeights_0=1
    ConstraintWeights_1:float=0
   ) -> (Pose) {boneToModify {boneName=base_earring_l}, constraintSetup [{targetBone {boneName=head}}], constraintWeights [1,0]}
   n20 = "Constraint - base_earring_r (P)"!(ComponentPose=@n19, Alpha=1, ConstraintWeights_0=1, ConstraintWeights_1:float=0) -> (Pose) {boneToModify {boneName=base_earring_r}, constraintSetup [{targetBone {boneName=head}}], constraintWeights [1,0]}
   Save cached pose 'predyn_cache'(Pose=Component To Local(ComponentPose=@n20)) {cacheName=predyn_cache}
  ]}
 EventGraph {nodes [
   Event BlueprintUpdateAnimation()
   > n1 = Sequence()
   > n2 = "Gravity Override State Function (Parameter Asset) (Nativized)"(
    CharacterPartAnimInstance=Self-Reference()
    GameTimeInSeconds=GetGameTimeInSeconds()
    Parameters=/BRCosmetics/Characters/Player/Male/Medium/Bodies/M_MED_SpadeStrut/Meshes/M_MED_SpadeStrut_Parameters
   )
   > Set GravityOverride(GravityOverride=@n2)
   n3 = Get SkelMesh()
   n4 = Get bRigidBodyEnabled()
   > Set Head Up(Head Up:float=float * float(
     A=GetAlphaFromReferenceTransformDelta(SkeletalMeshComponent=@n3, BoneName=spine_04, BaseName=head, InputRange=(X=15,Y=60), TwistAxis="0,0,1")
     B="To Float (Boolean)"(InBool=@n4)
    ) {is=title K2Node_PromotableOperator}) -> (Output_Get:float)
   n5 = Branch(execute=@n1.then_1, Condition=NOT Boolean(A=Get bIsInFrontEnd()))
   > Set RbanAlpha(RbanAlpha:double=1) -> (Output_Get:double)
   Set RbanAlpha(execute=@n5.else, RbanAlpha=0.3)
   Event OnCharacterCustomizationLoaded()
   > n6 = Cast To PlayerPawn_Athena(Object=TryGetPawnOwner())
   > "Set Face Mesh "("Face Mesh ":Skeletal Mesh Component Object Reference=GetSkeletalMeshForPartType(self=@n6, PartType=Head)) -> (Output_Get:Skeletal Mesh Component Object Reference)
   > "Set Body Mesh "("Body Mesh ":Skeletal Mesh Component Object Reference=GetSkeletalMeshForPartType(self=@n6)) -> (Output_Get:Skeletal Mesh Component Object Reference)
   Event BlueprintInitializeAnimation()
   > Set SkelMesh(SkelMesh=GetOwnerSkeletalMeshComponent())
  ]}
]
variables [
 "Body Mesh " {type=SkeletalMeshComponent}
 "Face Mesh " {type=SkeletalMeshComponent}
 GravityOverride {type=Vector}
 Head Up {type=number}
 RbanAlpha {type=number}
 SkelMesh {type=SkeletalMeshComponent}
]
